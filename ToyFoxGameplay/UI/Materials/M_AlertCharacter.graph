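# Translucent UI material that draws an animated alert character portrait over a glowing, framed backdrop, tinted by CharacterColor.
# Manual flipbook math steps the AlertPortrait 2x2 sheet at the FPS parameter on looping time.
# Panning noise scaled by Glitch In distorts the UVs shared by the portrait and a skewed trapezoid SDF mask.
# That mask shapes a vertical gradient glow, while an MSDF texture supplies the antialiased frame.
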Material /ToyFoxGameplay/UI/Materials/M_AlertCharacter
graph {
 properties {bAutomaticallySetUsageInEditor=false, blendMode=BLEND_Translucent, materialDomain=MD_UI, usedShadingModels=Default Lit}
 nodes [
  n1 = Constant3Vector() {constant=<rgba 1 1 1 0>}
  n2 = NamedRerouteUsage()
  n3 = NamedRerouteUsage()
  n4 = NamedRerouteUsage()
  n5 = VectorParameter() {parameterName=CharacterColor, defaultValue=<rgba 0.7835 0.1651 0.4564 1>}
  LinearInterpolate(A=Add(A=LinearInterpolate(A=@n1, B=@n2, Alpha=@n3), B=Multiply(A=@n4, B=@n5.RGB)), B=@n5.RGB, Alpha=NamedRerouteUsage())
  MaterialFunctionCall() {materialFunction=&FlipBook}
  Max(A=Max(A=NamedRerouteUsage(), B=NamedRerouteUsage()), B=NamedRerouteUsage())
  Multiply(A=Time(), B=Constant() {r=0.8})
  Multiply()
  n6 = MaterialFunctionCall() {materialFunction=&GetUserInterfaceUV}
  n7 = Power(Base=Add(A=ComponentMask(None=@n6.Normalized UV) {g=true}) {constB=0.1}) {constExponent=10}
  n8 = TextureCoordinate() {uTiling=4, vTiling=2}
  n9 = Multiply(A=Time()) {constB=2.5}
  n10 = Constant2Vector() {g=0.25}
  n11 = ComponentMask(None=TextureSample(UVs=Panner(Coordinate=@n8, Time=@n9, Speed=@n10)) {samplerType=SAMPLERTYPE_LinearGrayscale, texture=/ToyFoxGameplay/UI/Materials/Textures/T_UI_POI_Enter_Noise}.RGB) {g=true, r=true}
  n12 = ScalarParameter() {defaultValue=0.1, parameterName=Glitch In}
  n13 = Add(
   A=Add(A=@n6.Normalized UV, B=Constant2Vector() {g=0.03, r=0.05})
   B=Multiply(A=Subtract(A=@n11) {constB=0.5}, B=Multiply(A=@n12) {constB=0.1})
  )
  n14 = Constant2Vector() {g=4, r=-9}
  n15 = Constant() {r=0.7}
  n16 = MaterialFunctionCall(DistanceField=MaterialFunctionCall!(
    UVs=MaterialFunctionCall!(UV=@n13, Skew Angle=@n14) {materialFunction=/FortUILibrary/MaterialFunctions/Basics/MF_UI_Skew}.Skewed UVs
    TopWidth=@n15
    BottomWidth=Add(A=@n15) {constB=0.1}
    Height=Constant() {r=0.75}
   ) {materialFunction=/FortUILibrary/MaterialFunctions/Shapes/SDFs/MF_SDF_IsoscelesTrapezoid}.SDF) {materialFunction=&MF_UI_AntiAliasing}
  NamedRerouteDeclaration(None=Multiply(A=Multiply(A=@n7) {constB=2}, B=NamedRerouteDeclaration(None=@n16.Inner) {name=BackgroundMask, nodeColor=<rgba 0.1 0.0813 0.05 0.75>})) {name=BackgroundGlow, nodeColor=<rgba 0.0999 0.0823 0.0497 0.749>}
  n17 = Multiply(A=Time() {bOverride_Period=true, period=1.5}, B=ScalarParameter() {defaultValue=12, parameterName=FPS})
  n18 = Constant2Vector() {g=2, r=2}
  n19 = Divide(A=@n17, B=ComponentMask(None=@n18) {r=true})
  n20 = Add(A=@n13, B=AppendVector(A=Floor(None=@n17), B=Floor(None=@n19)))
  n21 = Divide(A=Constant() {r=1}, B=@n18)
  n22 = TextureSample(
   UVs=Frac(None=Multiply(A=@n20, B=@n21))
   Tex=TextureObjectParameter() {parameterName=AlertPortrait, texture=/ToyFoxGameplay/UI/Textures/T_ToyFox_Alert_BadGuy_Sheet}
  )
  NamedRerouteDeclaration(None=Multiply(A=@n22.A, B=NamedRerouteUsage())) {name=CharacterMask, nodeColor=<rgba 0.0999 0.0823 0.0497 0.749>}
  NamedRerouteDeclaration(None=@n22.RGB) {name=CharacterRGB, nodeColor=<rgba 0.1 0.0813 0.05 0.75>}
  n23 = Constant2Vector() {g=-0.37}
  n24 = TextureSample(UVs=MaterialFunctionCall(UV=Add(A=@n6.Normalized UV, B=@n23), TextureSize=Constant2Vector() {g=63, r=196}) {materialFunction=&MF_UI_UVScaling}.Scaled to Fit) {samplerType=SAMPLERTYPE_LinearColor, texture=/ToyFoxGameplay/UI/Materials/Textures/T_UI_CTABottom}
  NamedRerouteDeclaration(None=MaterialFunctionCall(
    DistanceField=MaterialFunctionCall!(MSDF RGB=@n24.RGB) {materialFunction=/FortUILibrary/MaterialFunctions/Shapes/Modifiers/MF_UI_MSDFtoSDF}.SDF
    Threshold=Constant() {r=0.5}
   ) {materialFunction=&MF_UI_AntiAliasing}.Outer) {name=BackgroundFrame, nodeColor=<rgba 0.0999 0.0823 0.0497 0.749>}
 ]
}
