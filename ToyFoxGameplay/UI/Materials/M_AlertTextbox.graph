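# Translucent UI material for an alert textbox, drawing a dark, partly see-through panel with a colored accent at its left edge.
# Skewed, 9-sliced UVs feed a FortUILibrary isosceles trapezoid function, whose Fill forms the panel shape.
# A steep power falloff on inverted U, masked by that fill, is tinted by the CharacterColor parameter over a near-black base.
# The same accent mask is added to the scaled-down fill, making the edge more opaque.
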Material /ToyFoxGameplay/UI/Materials/M_AlertTextbox
graph {
 properties {bAutomaticallySetUsageInEditor=false, blendMode=BLEND_Translucent, materialDomain=MD_UI, usedShadingModels=Default Lit}
 nodes [
  n1 = TextureCoordinate()
  n2 = ComponentMask(None=@n1) {r=true}
  n3 = Add(A=@n1, B=Constant2Vector())
  n4 = Constant2Vector() {r=-1}
  n5 = Constant() {r=0.7}
  n6 = Add(A=@n5) {constB=0.15}
  n7 = MaterialFunctionCall!(
   UVs=MaterialFunctionCall!(UVs=MaterialFunctionCall!(UV=@n3, Skew Angle=@n4) {materialFunction=/FortUILibrary/MaterialFunctions/Basics/MF_UI_Skew}.Skewed UVs) {materialFunction=/FortUILibrary/MaterialFunctions/Basics/MF_UI_9SliceUVs}.UVs
   TopWidth=@n6
   BottomWidth=@n5
   Height=@n6
   OutlineThickness
   OutlineAlignment
   GlowMin
   GlowMax
  ) {materialFunction=/FortUILibrary/MaterialFunctions/Shapes/MF_UI_IsoscelesTrapezoid}
  n8 = Multiply(A=Power(Base=OneMinus(None=@n2)) {constExponent=30}, B=@n7.Fill)
  Add(
   A=Constant3Vector() {constant=<rgba 0.0052 0.0052 0.0052 1>}
   B=Multiply(A=VectorParameter() {parameterName=CharacterColor, defaultValue=<rgba 0.7835 0.1651 0.4564 1>}.RGB, B=@n8)
  )
  Add(A=@n8, B=Multiply(A=@n7.Fill) {constB=0.6})
 ]
}
