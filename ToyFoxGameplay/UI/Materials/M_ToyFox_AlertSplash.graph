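# UI material for the ToyFox alert splash, drawing a banner with outline, colored accents and a lightning border.
# An MSDF banner texture yields antialiased fill, outline and offset bottom-edge accent, while a mirrored SDF texture forms the lightning.
# Circle SDFs add radial glows that tint the banner and lightning.
# Mirrored, rotated gradient steps create parameter-driven reveal wipes for banner and lightning, with banner inversion available.
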Material /ToyFoxGameplay/UI/Materials/M_ToyFox_AlertSplash
graph {
 properties {bAutomaticallySetUsageInEditor=false, blendMode=BLEND_AlphaComposite, materialDomain=MD_UI, usedShadingModels=Default Lit}
 nodes [
  n1 = VectorParameter() {group=Accents, parameterName=Accents - Color, sortPriority=0, defaultValue=<rgba 0.8714 0.0561 0.1651 1>}
  n2 = LinearInterpolate(B=@n1.RGB, Alpha=Saturate(None=NamedRerouteUsage()))
  n3 = VectorParameter() {group=Banner, parameterName=Banner - Outer Color, sortPriority=10, defaultValue=<rgba 0.0503 0.0137 0.0166 1>}
  n4 = VectorParameter() {group=Banner, parameterName=Banner - Inner Color, sortPriority=0, defaultValue=<rgba 0.1094 0.0568 0.0632 1>}
  n5 = LinearInterpolate(A=@n3.RGB, B=@n4.RGB, Alpha=NamedRerouteUsage())
  n6 = VectorParameter() {group=Lightning Border, parameterName=Lightning Accent - Color, sortPriority=40, defaultValue=<rgba 0.8714 0.0561 0.1651 1>}
  n7 = Multiply(A=Saturate(None=NamedRerouteUsage())) {constB=3}
  n8 = TextureCoordinate()
  n9 = Add(
   A=MaterialFunctionCall(
    UVs=MaterialFunctionCall!(UVs=@n8) {materialFunction=/FortUILibrary/MaterialFunctions/Basics/MF_UI_9SliceUVs}.UVs
    Texture Scale=Constant() {r=0.9}
   ) {materialFunction=&ScaleUVsByCenter}.UVs
   B=ComponentMask(None=VectorParameter() {group=Banner, parameterName=Banner Offset, defaultValue=<rgba 0 0 0 1>}.RGB) {g=true, r=true}
  )
  n10 = MaterialFunctionCall!(MSDF RGB=TextureSample(UVs=@n9) {samplerType=SAMPLERTYPE_LinearColor, texture=/ToyFoxGameplay/UI/Materials/Textures/T_UI_SplashBanner}.RGB) {materialFunction=/FortUILibrary/MaterialFunctions/Shapes/Modifiers/MF_UI_MSDFtoSDF}
  n11 = Constant() {r=0.65}
  n12 = MaterialFunctionCall(DistanceField=@n10.SDF, Threshold=@n11) {materialFunction=&MF_UI_AntiAliasing}
  n13 = Multiply(A=@n12.Outer, B=ScalarParameter() {defaultValue=1, group=Banner, parameterName=Banner Opacity, sortPriority=0})
  n14 = Constant() {r=0.5}
  n15 = MaterialFunctionCall(DistanceField=@n10.SDF, Threshold=@n14) {materialFunction=&MF_UI_AntiAliasing}
  n16 = ScalarParameter() {defaultValue=1, group=Outline, parameterName=Outline Opacity, sortPriority=0}
  n17 = LinearInterpolate(
   A=Multiply(A=NamedRerouteDeclaration(None=@n13) {name=Banner, nodeColor=<rgba 0.3359 0 1 1>}) {constB=0.75}
   B=NamedRerouteDeclaration(None=Multiply(A=@n15.Outer, B=@n16)) {name=Outline, nodeColor=<rgba 0 1 0.0156 1>}
  ) {constAlpha=0.7}
  n18 = NamedRerouteUsage()
  n19 = MaterialFunctionCall(Float2=MaterialFunctionCall!(UV=MaterialFunctionCall(UVs=@n8) {materialFunction=&MF_UI_9SliceUVs}.UVs, Mirror Y, Swizzle UV) {materialFunction=/FortUILibrary/MaterialFunctions/Basics/MF_UI_Symmetry_2x1_New}.Symmetry) {materialFunction=&BreakOutFloat2Components}
  n20 = Multiply(A=@n19.R) {constB=1.2}
  n21 = MaterialFunctionCall(
   UVs=Add(A=AppendVector(A=@n20, B=@n19.G), B=Constant2Vector() {r=-0.11})
   Texture Scale=Add(A=Constant2Vector() {g=0.4, r=1}, B=Constant() {r=0.08})
  ) {materialFunction=&ScaleUVsByCenter}
  n22 = ComponentMask(None=VectorParameter() {group=Lightning Border, parameterName=Lightning Offset, defaultValue=<rgba 0 0 0 1>}.RGB) {g=true, r=true}
  n23 = MaterialFunctionCall(
   DistanceField=TextureSample(UVs=Add(A=@n21.UVs, B=@n22)) {samplerType=SAMPLERTYPE_LinearColor, texture=/ToyFoxGameplay/UI/Materials/Textures/T_UI_LightningBorder}.R
   Threshold=@n14
  ) {materialFunction=&MF_UI_AntiAliasing}
  n24 = Multiply(
   A=NamedRerouteDeclaration(None=@n23.Outer) {name=Lightning, nodeColor=<rgba 0 0.625 1 1>}
   B=ScalarParameter() {defaultValue=0.02, group=Lightning Border, parameterName=Lightning Accent - Opacity, sliderMax=1, sortPriority=30}
  ) {constB=0.02}
  n25 = NamedRerouteUsage()
  Multiply(
   A=LinearInterpolate(A=Add(A=@n2, B=@n5), B=LinearInterpolate(B=@n6.RGB, Alpha=@n7) {constA=1}, Alpha=NamedRerouteUsage())
   B=Max(A=Multiply(A=@n17, B=@n18), B=Multiply(A=@n24, B=@n25))
  )
  n26 = Add(A=MaterialFunctionCall() {materialFunction=&GetUserInterfaceUV}.Normalized UV, B=Constant2Vector() {g=0.5})
  n27 = ScalarParameter() {defaultValue=1.01, group=Accents, parameterName=Accents - Inner Radial Glow Radius, sortPriority=10}
  n28 = Multiply(
   A=@n12.Outer
   B=Multiply(A=MaterialFunctionCall!(UVs=@n26, Radius=@n27) {materialFunction=/FortUILibrary/MaterialFunctions/Shapes/SDFs/MF_SDF_Circle}.SDF) {constB=-1}
  )
  n29 = VectorParameter() {group=Accents, parameterName=Accents - Bottom Offset, sortPriority=20, defaultValue=<rgba 0 0.01 0 1>}
  n30 = Add(A=@n9, B=ComponentMask(None=@n29.RGB) {g=true, r=true})
  n31 = MaterialFunctionCall(
   DistanceField=MaterialFunctionCall(MSDF RGB=TextureSample(UVs=@n30) {samplerType=SAMPLERTYPE_LinearColor, texture=&T_UI_SplashBanner}.RGB) {materialFunction=&MF_UI_MSDFtoSDF}.SDF
   Threshold=@n11
  ) {materialFunction=&MF_UI_AntiAliasing}
  n32 = Multiply(A=Saturate(None=Subtract(A=@n12.Outer, B=@n31.Outer))) {constB=0.3}
  NamedRerouteDeclaration(None=Multiply(
    A=Add(A=@n28, B=@n32)
    B=ScalarParameter() {defaultValue=1, group=Accents, parameterName=Accents - Bottom Opacity, sliderMax=1, sortPriority=20}
   )) {name=Accents, nodeColor=<rgba 1 0.6563 0 1>}
  n33 = MaterialFunctionCall() {materialFunction=&GetUserInterfaceUV}
  n34 = Constant2Vector() {g=0.38}
  n35 = Multiply(A=MaterialFunctionCall(
    UVs=Add(A=@n33.Normalized UV, B=@n34)
    Radius=ScalarParameter() {defaultValue=0.9, group=Lightning Border, parameterName=Lightning Accent - Scale, sortPriority=20}
   ) {materialFunction=&MF_SDF_Circle}.SDF) {constB=-1}
  n36 = ScalarParameter() {defaultValue=0.02, group=Lightning Border, parameterName=Lightning Accent - Opacity, sliderMax=1, sortPriority=30}
  NamedRerouteDeclaration(None=Multiply(A=@n23.Outer, B=Multiply(A=@n35, B=@n36))) {name=LightningAccent, nodeColor=<rgba 1 0 0.3047 1>}
  n37 = MaterialFunctionCall!(Repeat Count=Constant2Vector() {g=1, r=2}, texcoord=@n8) {materialFunction=/Game/Environments/Papaya/Materials/MaterialFunctions/MF_UVMirror}
  n38 = MaterialFunctionCall(
   UVs=@n37.Result
   "Rotation Angle (0-1)"=ScalarParameter() {defaultValue=0.1, group=Lightning Border, parameterName=Lightning Border - Reveal Angle, sortPriority=10}
  ) {materialFunction=&CustomRotator}
  NamedRerouteDeclaration(None=Step(
    Y=MaterialFunctionCall(UV Channel=@n38.Rotated Values) {materialFunction=&LinearGradient}.UGradient
    X=ScalarParameter() {defaultValue=1, group=Lightning Border, parameterName=Lightning Border - Reveal, sliderMax=1, sortPriority=0}
   ) {constX=0.5}) {name=Lightning_Reveal, nodeColor=<rgba 0 0.625 1 1>}
  n39 = ScalarParameter() {defaultValue=0.1, group=Banner, parameterName=Banner - Reveal Angle, sortPriority=10}
  n40 = Step(
   Y=MaterialFunctionCall(UV Channel=MaterialFunctionCall(UVs=@n37.Result, "Rotation Angle (0-1)"=@n39) {materialFunction=&CustomRotator}.Rotated Values) {materialFunction=&LinearGradient}.UGradient
   X=ScalarParameter() {group=Banner, parameterName=Banner - Reveal, sliderMax=1, sortPriority=0}
  ) {constX=0.5}
  n41 = ScalarParameter() {defaultValue=1, group=Banner, parameterName=Banner - Invert Reveal?, sliderMax=1, sortPriority=0}
  NamedRerouteDeclaration(None=LinearInterpolate(A=@n40, B=OneMinus(None=@n40), Alpha=Ceil(None=@n41))) {name=Banner_Reveal, nodeColor=<rgba 0 0.625 1 1>}
 ]
}
