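# Translucent UI material that draws the ToyFox alert splash icon: a head graphic over a colored glow, base and accent.
# A panning noise texture jitters the head UVs, with offset, scale and white-out parameters, and a trapezoid mask clips it.
# The head blends an alert texture's alpha with an MSDF silhouette; base and accent are antialiased MSDF shapes with adjustable thickness.
# The glow is an eased vertical gradient inside an SDF trapezoid, with switchable sine pulsing.
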Material /ToyFoxGameplay/UI/Materials/M_ToyFox_AlertSplash_Icon
graph {
 properties {bAutomaticallySetUsageInEditor=false, blendMode=BLEND_Translucent, materialDomain=MD_UI, usedShadingModels=Default Lit}
 nodes [
  Constant() {r=0.8}
  Constant() {r=0.2}
  n1 = VectorParameter() {group=Glow, parameterName=Glow Color, defaultValue=<rgba 0.8976 0.0497 0.1151 1>}
  n2 = LinearInterpolate(
   A=LinearInterpolate(
    A=@n1.RGB
    B=VectorParameter() {group=Accent, parameterName=Accent Color, defaultValue=<rgba 0.4948 0.0472 0.1072 1>}.RGB
    Alpha=NamedRerouteUsage()
   )
   B=VectorParameter() {group=Glow, parameterName=Glow Base Color, defaultValue=<rgba 0.8976 0.0497 0.1151 1>}.RGB
   Alpha=NamedRerouteUsage()
  )
  n3 = NamedRerouteUsage()
  n4 = MaterialFunctionCall() {materialFunction=&GetUserInterfaceUV}
  n5 = Time()
  n6 = TextureSample(UVs=Panner(Coordinate=@n4.Normalized UV, Time=Multiply(A=@n5) {constB=2.5}, Speed=Constant2Vector() {g=0.25})) {samplerType=SAMPLERTYPE_LinearGrayscale, texture=/ToyFoxGameplay/UI/Materials/Textures/T_UI_POI_Enter_Noise}
  n7 = Multiply(
   A=Subtract(A=ComponentMask(None=@n6.RGB) {g=true, r=true}) {constB=0.5}
   B=Multiply(A=ScalarParameter() {defaultValue=0.03, parameterName=Glitch In}) {constB=0.1}
  )
  n8 = Add(
   A=Add(A=Add(A=@n7, B=@n4.Normalized UV), B=Constant2Vector() {g=0.15})
   B=ComponentMask(None=VectorParameter() {group=Head, parameterName=Head Offset, sortPriority=10, defaultValue=<rgba 0 0 0 1>}.RGB) {g=true, r=true}
  )
  n9 = Constant2Vector() {g=0.5, r=0.65}
  n10 = ScalarParameter() {defaultValue=1, group=Head, parameterName=Head Scale, sortPriority=0}
  n11 = TextureSample(UVs=MaterialFunctionCall(UVs=@n8, Texture Scale=Multiply(A=@n9, B=@n10)) {materialFunction=&ScaleUVsByCenter}.UVs) {texture=/ToyFoxGameplay/UI/Materials/Textures/T_UI_AlertIcon}
  n12 = MaterialFunctionCall!(
   UVs=Add(A=@n4.Normalized UV, B=Constant2Vector() {g=0.22})
   TopWidth=Constant() {r=1.9}
   BottomWidth=Constant() {r=0.52}
   Height=Constant() {r=0.7}
   OutlineThickness
   OutlineAlignment
   GlowMin
   GlowMax
  ) {materialFunction=/FortUILibrary/MaterialFunctions/Shapes/MF_UI_IsoscelesTrapezoid}
  n13 = Multiply(A=@n11.RGBA, B=@n12.Fill)
  n14 = ComponentMask(None=@n13) {b=true, g=true, r=true}
  n15 = Constant3Vector() {constant=<rgba 1 1 1 0>}
  n16 = ScalarParameter() {group=Head, parameterName=Head White Out}
  n17 = Multiply(A=Add(A=@n9, B=Constant2Vector() {g=0.35, r=0.2125}) {constB=0.3}, B=@n10)
  n18 = MaterialFunctionCall!(MSDF RGB=TextureSample(UVs=MaterialFunctionCall(UVs=@n8, Texture Scale=@n17) {materialFunction=&ScaleUVsByCenter}.UVs) {samplerType=SAMPLERTYPE_LinearColor, texture=/ToyFoxGameplay/UI/Materials/Textures/T_UI_ToyFoxSDF}.RGB) {materialFunction=/FortUILibrary/MaterialFunctions/Shapes/Modifiers/MF_UI_MSDFtoSDF}
  n19 = Constant() {r=0.5}
  n20 = Multiply(
   A=LinearInterpolate(
    A=MaterialFunctionCall(DistanceField=@n18.SDF, Threshold=@n19) {materialFunction=&MF_UI_AntiAliasing}.Outer
    B=@n11.A
   )
   B=@n12.Fill
  )
  n21 = Add(A=Constant2Vector() {g=-0.1}, B=@n4.Normalized UV)
  n22 = MaterialFunctionCall(
   DistanceField=MaterialFunctionCall(MSDF RGB=TextureSample(UVs=@n21) {samplerType=SAMPLERTYPE_LinearColor, texture=/ToyFoxGameplay/UI/Materials/Textures/T_UI_GlowBase}.RGB) {materialFunction=&MF_UI_MSDFtoSDF}.SDF
   Threshold=LinearInterpolate(A=Constant() {r=0.5}, Alpha=ScalarParameter() {group=Glow, parameterName=Glow Base Thickness Offset}) {constB=0.6}
  ) {materialFunction=&MF_UI_AntiAliasing}
  n23 = ScalarParameter() {defaultValue=1, group=Glow, parameterName=Glow Base Opacity}
  n24 = Add(A=@n4.Normalized UV, B=Constant2Vector() {g=-0.2})
  n25 = MaterialFunctionCall(
   DistanceField=MaterialFunctionCall(MSDF RGB=TextureSample(UVs=@n24) {samplerType=SAMPLERTYPE_LinearColor, texture=/ToyFoxGameplay/UI/Materials/Textures/T_UI_GlowAccent}.RGB) {materialFunction=&MF_UI_MSDFtoSDF}.SDF
   Threshold=LinearInterpolate(A=Constant() {r=0.5}, Alpha=ScalarParameter() {group=Accent, parameterName=Accent Thickness Offset}) {constB=0.87}
  ) {materialFunction=&MF_UI_AntiAliasing}
  n26 = ScalarParameter() {defaultValue=1, parameterName=Accent Opacity}
  n27 = Max(
   A=NamedRerouteDeclaration(None=Multiply(A=@n22.Outer, B=@n23)) {name=GlowBase, nodeColor=<rgba 0.75 0.6063 0.375 0.75>}
   B=NamedRerouteDeclaration(None=Multiply(A=@n25.Outer, B=@n26)) {name=IconAccent, nodeColor=<rgba 0.75 0.6063 0.375 0.75>}
  )
  Multiply(
   A=LinearInterpolate(
    A=LinearInterpolate(A=@n2, B=@n1.RGB, Alpha=@n3)
    B=LinearInterpolate(A=@n14, B=@n15, Alpha=@n16) {constAlpha=0}
    Alpha=ComponentMask(None=@n13) {a=true}
   )
   B=Add(A=@n3, B=Max(A=@n20, B=@n27))
  )
  n28 = Add(A=@n4.Normalized UV) {constB=-0.2}
  n29 = MaterialFunctionCall("Value (0-1)"=ScalarParameter() {defaultValue=1, group=Glow, parameterName=Glow Strength}) {materialFunction=&MF_UI_EaseCurves}
  n30 = Multiply(A=Power(Base=ComponentMask(None=@n28) {g=true}, Exp=LinearInterpolate(Alpha=@n29.EaseInOutSine) {constA=12, constAlpha=0, constB=4}) {constExponent=4}) {constB=30}
  n31 = Constant2Vector() {g=-0.005}
  n32 = MaterialFunctionCall(DistanceField=MaterialFunctionCall!(UVs=Add(A=@n4.Normalized UV, B=@n31), TopWidth=Constant() {r=1}, BottomWidth=Constant() {r=0.55}, Height=Constant() {r=0.3}) {materialFunction=/FortUILibrary/MaterialFunctions/Shapes/SDFs/MF_SDF_IsoscelesTrapezoid}.SDF) {materialFunction=&MF_UI_AntiAliasing}
  n33 = ScalarParameter() {defaultValue=1, group=Glow, parameterName=Glow Opacity, sliderMax=1}
  n34 = Add(A=MaterialFunctionCall(Speed=Constant() {r=1}) {materialFunction=&DebugTimeSine}.Result) {constB=0.75}
  n35 = ScalarParameter() {defaultValue=1, parameterName=PulseSwitch}
  NamedRerouteDeclaration(None=Multiply(A=Multiply(A=@n30, B=@n32.Inner), B=LinearInterpolate(A=@n33, B=@n34, Alpha=@n35))) {name=Glow, nodeColor=<rgba 0 1 0.5781 1>}
 ]
}
