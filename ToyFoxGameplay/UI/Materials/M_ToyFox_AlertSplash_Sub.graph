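# Translucent UI material for a ToyFox alert splash sub-banner, combining a banner shape, edge backers and splash accents.
# MSDF banner and accent textures plus an edge-backer distance field become anti-aliased masks, with symmetry functions mirroring the backer and accents.
# A parameterized step over a mirrored, rotated gradient reveals the accents, with optional inversion.
# Masks combine through Max, while rerouted masks blend accent and banner colors, including a circle-SDF gradient declaration.
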Material /ToyFoxGameplay/UI/Materials/M_ToyFox_AlertSplash_Sub
graph {
 properties {bAutomaticallySetUsageInEditor=false, blendMode=BLEND_Translucent, materialDomain=MD_UI, usedShadingModels=Default Lit}
 nodes [
  n1 = VectorParameter() {parameterName=Accent Color, defaultValue=<rgba 0.0189 0.0498 0.0781 1>}
  n2 = VectorParameter() {parameterName=Banner Color, defaultValue=<rgba 0.6172 0.8632 0.8879 1>}
  n3 = NamedRerouteUsage()
  n4 = NamedRerouteUsage()
  LinearInterpolate(A=LinearInterpolate(A=LinearInterpolate(A=@n1.RGB, B=@n2.RGB, Alpha=@n3), B=@n2.RGB, Alpha=Saturate(None=@n4)), Alpha=NamedRerouteUsage())
  n5 = MaterialFunctionCall(UV=MaterialFunctionCall() {materialFunction=&GetUserInterfaceUV}.Normalized UV) {materialFunction=&MF_UI_UVScaling}
  n6 = Constant2Vector()
  n7 = AppendVector(A=Constant() {r=-0.001}, B=@n6.G)
  n8 = Add(
   A=MaterialFunctionCall!(UV=Add(A=@n5.Scaled To Fill, B=@n7), Mirror Y, Swizzle UV) {materialFunction=/FortUILibrary/MaterialFunctions/Basics/MF_UI_Symmetry_2x1_New}.Symmetry
   B=Constant2Vector() {g=-0.008, r=-0.2}
  )
  n9 = Constant2Vector() {g=0.0875, r=0.2}
  n10 = Constant() {r=0.5}
  n11 = MaterialFunctionCall(
   DistanceField=TextureSample(UVs=MaterialFunctionCall(UVs=@n8, Texture Scale=@n9) {materialFunction=&ScaleUVsByCenter}.UVs) {samplerType=SAMPLERTYPE_LinearColor, texture=/ToyFoxGameplay/UI/Materials/Textures/T_UI_EdgeBacker}.R
   Threshold=@n10
  ) {materialFunction=&MF_UI_AntiAliasing}
  n12 = MaterialFunctionCall(UVs=Add(A=@n5.Scaled To Fill, B=@n6), Texture Scale=Constant() {r=0.44}) {materialFunction=&ScaleUVsByCenter}
  n13 = MaterialFunctionCall(
   DistanceField=MaterialFunctionCall!(MSDF RGB=TextureSample(UVs=@n12.UVs) {samplerType=SAMPLERTYPE_LinearColor, texture=/ToyFoxGameplay/UI/Materials/Textures/T_UI_SubBanner}.RGB) {materialFunction=/FortUILibrary/MaterialFunctions/Shapes/Modifiers/MF_UI_MSDFtoSDF}.SDF
   Threshold=@n10
  ) {materialFunction=&MF_UI_AntiAliasing}
  n14 = NamedRerouteDeclaration(None=@n13.Outer) {name=Banner, nodeColor=<rgba 0 1 0.25 1>}
  n15 = ScalarParameter() {defaultValue=0.7, group=Banner, parameterName=Banner Opacity, sliderMax=1}
  n16 = Add(A=@n5.Scaled To Fill, B=@n6)
  n17 = TextureSample(UVs=MaterialFunctionCall(
    UVs=MaterialFunctionCall(UV=@n16) {materialFunction=&MF_UI_Symmetry_2x1_New}.Symmetry
    Texture Scale=Constant2Vector() {g=0.2, r=0.25}
   ) {materialFunction=&ScaleUVsByCenter}.UVs) {samplerType=SAMPLERTYPE_LinearColor, texture=/ToyFoxGameplay/UI/Materials/Textures/T_UI_SplashAccent}
  n18 = MaterialFunctionCall!(Repeat Count=Constant2Vector() {g=1, r=2}, texcoord=TextureCoordinate()) {materialFunction=/Game/Environments/Papaya/Materials/MaterialFunctions/MF_UVMirror}
  n19 = ScalarParameter() {defaultValue=0.1, group=Accents, parameterName=Accents - Reveal Angle, sortPriority=10}
  n20 = Step(
   Y=MaterialFunctionCall(UV Channel=MaterialFunctionCall(UVs=@n18.Result, "Rotation Angle (0-1)"=@n19) {materialFunction=&CustomRotator}.Rotated Values) {materialFunction=&LinearGradient}.UGradient
   X=ScalarParameter() {defaultValue=1, group=Accents, parameterName=Accents - Reveal, sliderMax=1, sortPriority=0}
  ) {constX=0.5}
  n21 = OneMinus(None=@n20)
  n22 = Ceil(None=ScalarParameter() {defaultValue=1, group=Accents, parameterName=Accents - Invert Reveal?, sliderMax=1, sortPriority=0})
  n23 = Multiply(
   A=MaterialFunctionCall(
    DistanceField=MaterialFunctionCall(MSDF RGB=@n17.RGB) {materialFunction=&MF_UI_MSDFtoSDF}.SDF
    Threshold=@n10
   ) {materialFunction=&MF_UI_AntiAliasing}.Outer
   B=NamedRerouteDeclaration(None=LinearInterpolate(A=@n20, B=@n21, Alpha=@n22)) {name=Accents_Reveal, nodeColor=<rgba 0 0.625 1 1>}
  )
  Max(
   A=Max(A=NamedRerouteDeclaration(None=@n11.Outer) {name=BannerAccents, nodeColor=<rgba 1 0 0.0703 1>}, B=Multiply(A=@n14, B=@n15) {constB=0.7})
   B=Multiply(A=NamedRerouteDeclaration(None=@n23) {name=Accents, nodeColor=<rgba 0 0.3906 1 1>}) {constB=0.5}
  )
  n24 = MaterialFunctionCall!(UVs=Add(A=@n8, B=Constant2Vector() {g=0.15, r=0.1}), Radius=Constant() {r=0.5}) {materialFunction=/FortUILibrary/MaterialFunctions/Shapes/SDFs/MF_SDF_Circle}
  NamedRerouteDeclaration(None=Multiply(A=Multiply(A=@n24.SDF) {constB=-1}, B=@n13.Outer)) {name=Gradients, nodeColor=<rgba 0.7109 1 0 1>}
 ]
}
