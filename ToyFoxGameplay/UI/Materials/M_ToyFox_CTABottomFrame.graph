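# Translucent UI material for ToyFox, named as the bottom frame of a CTA element, built from a distance-field frame texture.
# The CTA frame texture's RGB is converted from MSDF to SDF, then anti-aliased at a mid threshold.
# The outer mask is scaled down by a constant, with no link to a material output shown.
# A UV scaling function is present but unconnected, so the texture samples with default UVs.
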Material /ToyFoxGameplay/UI/Materials/M_ToyFox_CTABottomFrame
graph {
 properties {bAutomaticallySetUsageInEditor=false, blendMode=BLEND_Translucent, materialDomain=MD_UI, usedShadingModels=Default Lit}
 nodes [
  MaterialFunctionCall() {materialFunction=&MF_UI_UVScaling}
  n1 = TextureSample() {samplerType=SAMPLERTYPE_LinearColor, texture=/ToyFoxGameplay/UI/Materials/Textures/T_UI_CTAFrame}
  Multiply(A=MaterialFunctionCall(
    DistanceField=MaterialFunctionCall!(MSDF RGB=@n1.RGB) {materialFunction=/FortUILibrary/MaterialFunctions/Shapes/Modifiers/MF_UI_MSDFtoSDF}.SDF
    Threshold=Constant() {r=0.5}
   ) {materialFunction=&MF_UI_AntiAliasing}.Outer) {constB=0.4}
 ]
}
