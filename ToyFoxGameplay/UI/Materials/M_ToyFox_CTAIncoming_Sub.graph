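# UI material for the ToyFox "CTA Incoming" sub-element, drawing a procedural rounded, outlined shape in dark and orange tones.
# SDF boxes on 9-slice and fill-scaled UVs build fill, outline, angled inner-cut and highlight-shine masks.
# Named reroutes carry masks into layered color blends, while alpha-composite opacity combines fill, outline and highlight.
# LowHighSwitch fades in highlights and orange gradient, with separate master opacities for inner shape and outline.
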Material /ToyFoxGameplay/UI/Materials/M_ToyFox_CTAIncoming_Sub
graph {
 properties {bAutomaticallySetUsageInEditor=false, blendMode=BLEND_AlphaComposite, materialDomain=MD_UI, usedShadingModels=Default Lit}
 nodes [
  n1 = MaterialFunctionCall() {materialFunction=&GetUserInterfaceUV}
  n2 = MaterialFunctionCall(UV=@n1.Normalized UV) {materialFunction=&MF_UI_UVScaling}
  n3 = MaterialFunctionCall!(
   UVs=@n2.Scaled To Fill
   Size=Constant2Vector() {g=0.1, r=1}
   CornerRoundness=Constant4Vector() {constant=<rgba 0 0 0.05 0.09>}.RGBA
   KeepAspectRatio
   PixelSize
  ) {materialFunction=/FortUILibrary/MaterialFunctions/Shapes/SDFs/MF_SDF_Box}
  n4 = MaterialFunctionCall(Float2=@n1.Normalized UV) {materialFunction=&BreakOutFloat2Components}
  n5 = NamedRerouteDeclaration(None=Multiply(
    A=MaterialFunctionCall(DistanceField=@n3.SDF) {materialFunction=&MF_UI_AntiAliasing}.Inner
    B=Subtract(A=@n4.R) {constB=0.2}
   )) {name=Mask_Highlight, nodeColor=<rgba 0.75 0.6063 0.375 0.75>}
  n6 = ScalarParameter() {desc=default: 0, parameterName=LowHighSwitch}
  n7 = ComponentMask(None=MaterialFunctionCall(UVs=@n1.Normalized UV, "Rotation Angle (0-1)"=Constant() {r=-0.02}) {materialFunction=&CustomRotator}.Rotated Values) {r=true}
  n8 = Constant() {r=0.5}
  n9 = MaterialFunctionCall(
   UVs=MaterialFunctionCall!(UVs=@n1.Normalized UV) {materialFunction=/FortUILibrary/MaterialFunctions/Basics/MF_UI_9SliceUVs}.UVs
   Size=@n8
   CornerRoundness=@n8
  ) {materialFunction=&MF_SDF_Box}
  n10 = Max(A=Subtract(A=OneMinus(None=@n7), B=Constant() {r=1}) {constB=0.465}, B=@n9.SDF)
  n11 = ScalarParameter() {defaultValue=0.5, parameterName=Inner Shape Master Opacity}
  n12 = NamedRerouteDeclaration(None=Multiply(
    A=MaterialFunctionCall(DistanceField=@n10) {materialFunction=&MF_UI_AntiAliasing}.Inner
    B=@n11
   )) {name=Mask_InnerCircle, nodeColor=<rgba 0.75 0.6063 0.375 0.75>}
  n13 = Multiply(A=Multiply(
    A=MaterialFunctionCall(DistanceField=@n9.SDF) {materialFunction=&MF_UI_AntiAliasing}.Inner
    B=@n11
   )) {constB=0.9}
  n14 = MaterialFunctionCall!(DistanceField=@n9.SDF, Threshold, Thickness=Constant() {r=0.08}, Alignment=Constant() {r=1}, Smoothing) {materialFunction=/FortUILibrary/MaterialFunctions/Shapes/Modifiers/MF_UI_ShapeOutline}
  n15 = ScalarParameter() {defaultValue=1, parameterName=Outline Master Opacity}
  n16 = Multiply(A=NamedRerouteDeclaration(None=Multiply(A=@n14.Outline, B=@n15)) {name=Mask_Outline, nodeColor=<rgba 0.75 0.6063 0.375 0.75>}) {constB=0.5}
  Add(A=Saturate(None=LinearInterpolate(B=@n5, Alpha=@n6) {constB=0}), B=Max(A=Multiply(A=@n12) {constB=0}, B=Max(A=@n13, B=@n16)))
  n17 = NamedRerouteUsage()
  n18 = ScalarParameter() {desc=default: 0, parameterName=LowHighSwitch}
  n19 = LinearInterpolate(B=Constant() {r=0.4}, Alpha=NamedRerouteUsage()) {constB=0.4}
  n20 = LinearInterpolate(
   A=Constant3Vector() {constant=<rgba 0.5182 0.1005 0.0079 1>}
   B=Constant3Vector() {constant=<rgba 0.7969 0.1691 0 1>}
   Alpha=ComponentMask(None=TextureCoordinate()) {r=true}
  )
  n21 = LinearInterpolate(A=Constant3Vector() {constant=<rgba 0 0 0 1>}, B=@n20, Alpha=ScalarParameter() {desc=default: 0, parameterName=LowHighSwitch})
  n22 = NamedRerouteUsage()
  n23 = LinearInterpolate(B=NamedRerouteUsage(), Alpha=@n18)
  Add(
   A=LinearInterpolate(B=Constant3Vector() {constant=<rgba 1 0.5149 0 1>}, Alpha=LinearInterpolate(B=@n17, Alpha=@n18))
   B=Add(A=LinearInterpolate(A=@n19, B=@n21, Alpha=@n22), B=LinearInterpolate(B=@n20, Alpha=@n23))
  )
  Constant2Vector() {g=0.9, r=0.6}
  Constant() {r=0.465}
  MaterialFunctionCall(UV=@n1.Normalized UV) {materialFunction=&MF_UI_UVScaling}
  n24 = Add(A=Constant2Vector() {r=-0.2}, B=@n2.Scaled To Fill)
  n25 = Constant2Vector() {g=1, r=0.5}
  n26 = Saturate(None=Multiply(A=MaterialFunctionCall(UVs=@n24, Size=@n25) {materialFunction=&MF_SDF_Box}.SDF) {constB=-2})
  NamedRerouteDeclaration(None=Multiply(A=LinearInterpolate(B=@n26, Alpha=@n6) {constB=0}, B=@n5)) {name=Mask_HighlightShine, nodeColor=<rgba 0.75 0.6063 0.375 0.75>}
 ]
}
