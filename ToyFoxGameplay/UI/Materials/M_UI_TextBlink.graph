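# Additive UI material that draws a blinking, cursor-like bar and a maskable rounded box, apparently for text effects.
# The narrow bar is shifted by Blinky X Movement and pulsed by a time sine weighted by Blinky.
# A Mask-driven sliding box intersects a fixed one through fill multiply and SDF max, then combines with the bar by max.
# Both shapes get a faint antialiased SDF glow, and Flash blends white toward warm yellow.
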Material /ToyFoxGameplay/UI/Materials/M_UI_TextBlink
graph {
 properties {bAutomaticallySetUsageInEditor=false, blendMode=BLEND_Additive, materialDomain=MD_UI, usedShadingModels=Default Lit}
 nodes [
  Add()
  LinearInterpolate(A=Constant() {r=1}, B=Constant3Vector() {constant=<rgba 1 0.9755 0.3898 1>}, Alpha=ScalarParameter() {parameterName=Flash})
  n1 = TextureCoordinate()
  n2 = AppendVector(A=ScalarParameter() {defaultValue=0.296, parameterName=Blinky X Movement}, B=Constant())
  n3 = MaterialFunctionCall(
   UVs=MaterialFunctionCall(UV=Add(A=@n1, B=@n2)) {materialFunction=&MF_UI_UVScaling}.Scaled to Fit
   Size=Constant2Vector() {g=0.5, r=0.1}
   CornerRoundness=Constant() {r=0.05}
  ) {materialFunction=&MF_UI_Box}
  n4 = Constant() {r=-0.308}
  n5 = Constant() {r=0.856}
  n6 = Add(
   A=@n3.Fill
   B=Multiply(A=MaterialFunctionCall(DistanceField=@n3.SDF, Threshold=@n4, Smoothing=@n5) {materialFunction=&MF_UI_AntiAliasing}.Inner) {constB=0.2}
  )
  n7 = MaterialFunctionCall(Speed=Constant() {r=5}) {materialFunction=&DebugTimeSine}
  n8 = Constant() {r=0.05}
  n9 = StaticBool() {value=true}
  n10 = MaterialFunctionCall(Size=AppendVector(A=Constant() {r=0.524}, B=Constant() {r=0.5}), CornerRoundness=@n8, KeepAspectRatio=@n9) {materialFunction=&MF_UI_Box}
  n11 = Multiply(A=@n10.Fill)
  n12 = LinearInterpolate(Alpha=ScalarParameter() {defaultValue=0.5, parameterName=Mask}) {constA=1, constB=-1}
  n13 = Constant()
  n14 = MaterialFunctionCall(
   UVs=Add(A=TextureCoordinate(), B=AppendVector(A=@n12, B=@n13))
   Size=AppendVector(A=Constant() {r=0.524}, B=Constant() {r=0.5})
   CornerRoundness=@n8
   KeepAspectRatio=@n9
  ) {materialFunction=&MF_UI_Box}
  n15 = Multiply(A=@n14.Fill)
  n16 = MaterialFunctionCall(DistanceField=Max(A=@n10.SDF, B=@n14.SDF), Threshold=Constant() {r=-0.308}, Smoothing=Constant() {r=0.856}) {materialFunction=&MF_UI_AntiAliasing}
  Max(
   A=LinearInterpolate(A=@n6, B=Multiply(A=@n6, B=@n7.Result), Alpha=ScalarParameter() {defaultValue=1, parameterName=Blinky})
   B=Add(A=Multiply(A=@n11, B=@n15), B=Multiply(A=@n16.Inner) {constB=0.2})
  )
 ]
}
