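# M_UI_ToyFox_HardEdgedGlow is an additive UI material that builds a glow for ToyFox UI elements.
# LinearGradient U and V are folded into center-peaked falloffs, shaped by Power, with U boosted and saturated for hard edges.
# Their product blends two orange "Big Glow Color" parameters, and separately black to gray.
# The ColorSwitcher scalar interpolates between the colored and grayscale results.
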Material /ToyFoxGameplay/UI/Materials/M_UI_ToyFox_HardEdgedGlow
graph {
 properties {bAutomaticallySetUsageInEditor=false, blendMode=BLEND_Additive, materialDomain=MD_UI, usedShadingModels=Default Lit}
 nodes [
  n1 = MaterialFunctionCall() {materialFunction=&LinearGradient}
  n2 = Subtract(A=@n1.VGradient) {constB=0.5}
  n3 = OneMinus(None=Abs(None=Multiply(A=@n2) {constB=2}))
  n4 = Multiply(A=Subtract(A=@n1.UGradient) {constB=0.5}) {constB=2}
  n5 = Power(Base=OneMinus(None=Abs(None=@n4)))
  n6 = Multiply(A=Saturate(None=Power(Base=@n3)), B=Saturate(None=Multiply(A=@n5) {constB=30}))
  LinearInterpolate(
   A=LinearInterpolate(
    A=VectorParameter() {parameterName=Big Glow Color 01, defaultValue=<rgba 1 0.2831 0 1>}.RGB
    B=VectorParameter() {parameterName=Big Glow Color 02, defaultValue=<rgba 1 0.3813 0 1>}.RGB
    Alpha=@n6
   )
   B=LinearInterpolate(A=Constant(), B=Constant() {r=0.5}, Alpha=@n6)
   Alpha=ScalarParameter() {parameterName=ColorSwitcher}
  )
 ]
}
