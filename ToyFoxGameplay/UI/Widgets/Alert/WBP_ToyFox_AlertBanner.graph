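# ToyFox HUD alert banner widget that shows tag-driven notifications received from the game state component.
# Filtered alerts look up a data table row by gameplay tag, then switch between Splash (Megazord), Incoming (Queen detected) and default portrait banner styles.
# Async-loaded stingers register with the music event subsystem; the portrait style also plays chance-based, delayed voice-over.
# Show animates material parameters, a timer triggers Hide, and start/finish delegates report alert state.
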WidgetBlueprint /ToyFoxGameplay/UI/Widgets/Alert/WBP_ToyFox_AlertBanner
animations [
 Show {
  binding_count=13
  length_seconds=1.45
  object=&self:Show
  sequence {
   bindings [
    Image_IncomingIcon {
     id=#1
     possessed_class=Image
     tracks [{
       type=MovieSceneWidgetMaterialTrack
       sections [{
         type=MovieSceneParameterSection
         end_frame=43
         start_frame=0
         channels [
          Glitch In {
           type=MovieSceneScriptingFloatChannel
           keys [[0,0],[3.001,0],[6.001,0.11],[6.501,0],[8.5005,0],[9.0005,-0.1],[10.5005,0],[11.0005,0.23],[13.0005,0],[13.5005,-0.11],[15.5005,0],[43.5,0]]
          }
          Bracket Show {type=MovieSceneScriptingFloatChannel, keys [[5.001,0],[12.5005,1],[43.5,1]]}
          Speed Up {type=MovieSceneScriptingFloatChannel, keys [[0.001,1],[2.001,0],[43.5,0]]}
          Gradient Front Reveal {type=MovieSceneScriptingFloatChannel, keys [[4.501,1],[8.001,0],[43.5,0]]}
          Triangle Scale {type=MovieSceneScriptingFloatChannel, keys [[0.001,0],[4.501,0.35],[43.5,0.35]]}
          "! Scale" {type=MovieSceneScriptingFloatChannel, keys [[0.001,0.9],[4.501,0],[43.5,0]]}
          Gradient Front Intensity {type=MovieSceneScriptingFloatChannel, keys [[4.501,0],[9.5005,0.9],[14.0005,0],[43.5,0]]}
          Gradient Opacity {type=MovieSceneScriptingFloatChannel, keys [[4.501,0],[7.001,1],[43.5,1]]}
          "Intro ! Switch" {type=MovieSceneScriptingFloatChannel, keys [[0,1],[4.501,0],[43.5,0]]}
          "! Translucency" {type=MovieSceneScriptingFloatChannel, keys [[0,0],[4.501,0.8],[43.5,0.8]]}
          Triangle Master Opacity {type=MovieSceneScriptingFloatChannel, keys [[0,0],[4.501,1],[43.5,1]]}
          High Level {type=MovieSceneScriptingFloatChannel, const [1,2]}
          Lightning - Leading Edge {type=MovieSceneScriptingFloatChannel, keys [[0,0],[15,0],[16,0.3],[18,0.4],[19,0.3],[23,1]]}
          Lightning - Trailing Edge {type=MovieSceneScriptingFloatChannel, thinned_from=9, keys [[0,0],[21,0],[27,0.8],[30,0.9],[32,0.85],[34,0.9],[36,0.92],[38,1]]}
         ]
        }]
      }]
    }
    Overlay_IncomingBannerWrapper {
     id=#2
     possessed_class=Overlay
     tracks [
      {
       type=MovieSceneFloatTrack
       display_name=Render Opacity
       sections [{
         type=MovieSceneFloatSection
         end_frame=43
         start_frame=0
         channels [None {type=MovieSceneScriptingFloatChannel, keys [[0,0],[28,1],[43.5,1]]}]
        }]
      }
      {
       type=MovieSceneByteTrack
       display_name=Visibility
       sections [{type=MovieSceneByteSection, end_frame=1, start_frame=0, channels [None {type=MovieSceneScriptingByteChannel, keys [[0,2],[1,4]]}]}]
      }
     ]
    }
    Image_IncomingBacker {
     id=#3
     possessed_class=Image
     tracks [{
       type=MovieSceneWidgetMaterialTrack
       sections [{
         type=MovieSceneParameterSection
         end_frame=43
         start_frame=0
         channels [
          SignBacker_Opacity {type=MovieSceneScriptingFloatChannel, const [0.4,2]}
          Outline_Opacity {type=MovieSceneScriptingFloatChannel, keys [[11,0],[15.5,2],[32,0.2],[43.5,0.2]]}
          Dots Opacity {type=MovieSceneScriptingFloatChannel, default=0, keys [[0,0],[0.5,1],[1.5,0],[3.5,1],[4.5,0.8],[6.5,0],[9,1],[43.5,1]]}
          Stripes Opacity {type=MovieSceneScriptingFloatChannel, const [1,2]}
          Background Reveal {type=MovieSceneScriptingFloatChannel, keys [[5.5,0],[19,1],[43.5,1]]}
          Outline Reveal {type=MovieSceneScriptingFloatChannel, keys [[3,0],[27,1],[43.5,1]]}
          Backer_Opacity {type=MovieSceneScriptingFloatChannel, keys [[0,0],[34,0.8],[43.5,0.8]]}
          Dots Scale {type=MovieSceneScriptingFloatChannel, thinned_from=5, keys [[0,1],[2.5,1],[3,0],[43.5,0]]}
          Bottom Outline Reveal {type=MovieSceneScriptingFloatChannel, keys [[3.5,1],[24.5,0],[43.5,0]]}
         ]
        }]
      }]
    }
    Image_SubBanner {
     id=#4
     possessed_class=Image
     tracks [
      {
       type=MovieSceneWidgetMaterialTrack
       sections [{
         type=MovieSceneParameterSection
         end_frame=43
         start_frame=0
         channels [
          Accents - Reveal {type=MovieSceneScriptingFloatChannel, keys [[0,0.604],[21,0.604],[24,0.38],[43,0.38]]}
          Accents - Reveal Angle {type=MovieSceneScriptingFloatChannel, const [0.003,1]}
          Accents - Invert Reveal? {type=MovieSceneScriptingFloatChannel, const [1,1]}
          Banner Opacity {type=MovieSceneScriptingFloatChannel, keys [[0,1],[20,1],[23,0.7],[43,0.7]]}
          Accent ColorR {type=MovieSceneScriptingFloatChannel, thinned_from=5, keys [[0,1],[20,1],[22,0.0189],[43,0.0189]]}
          Accent ColorG {type=MovieSceneScriptingFloatChannel, thinned_from=5, keys [[0,1],[20,1],[22,0.0498],[43,0.0498]]}
          Accent ColorB {type=MovieSceneScriptingFloatChannel, thinned_from=5, keys [[0,1],[20,1],[22,0.0781],[43,0.0781]]}
          Accent ColorA {type=MovieSceneScriptingFloatChannel, const [1,5]}
          Banner ColorR {type=MovieSceneScriptingFloatChannel, thinned_from=5, keys [[0,1],[20,1],[22,0.6172],[43,0.6172]]}
          Banner ColorG {type=MovieSceneScriptingFloatChannel, thinned_from=5, keys [[0,1],[20,1],[22,0.8632],[43,0.8632]]}
          Banner ColorB {type=MovieSceneScriptingFloatChannel, thinned_from=5, keys [[0,1],[20,1],[22,0.8879],[43,0.8879]]}
          Banner ColorA {type=MovieSceneScriptingFloatChannel, const [1,5]}
         ]
        }]
      }
      {
       type=MovieSceneFloatTrack
       display_name=Render Opacity
       sections [{
         type=MovieSceneFloatSection
         end_frame=43
         start_frame=0
         channels [None {type=MovieSceneScriptingFloatChannel, keys [[0,0],[17,0],[20,1],[43,1]]}]
        }]
      }
     ]
    }
    Overlay_SplashBannerWrapper {
     id=#5
     possessed_class=Overlay
     tracks [
      {
       type=MovieScene2DTransformTrack
       sections [{
         type=MovieScene2DTransformSection
         end_frame=43
         start_frame=0
         channels [
          Translation.X {type=MovieSceneScriptingFloatChannel, const [0,5]}
          Translation.Y {type=MovieSceneScriptingFloatChannel, keys [[0,0],[11,0],[12,4],[16,0],[43,0]]}
          Angle {type=MovieSceneScriptingFloatChannel, const [0,1]}
          Scale.X {type=MovieSceneScriptingFloatChannel, const [0.8,1]}
          Scale.Y {type=MovieSceneScriptingFloatChannel, const [0.8,1]}
          Shear.X {type=MovieSceneScriptingFloatChannel, const [0,1]}
          Shear.Y {type=MovieSceneScriptingFloatChannel, const [0,1]}
         ]
        }]
      }
      {
       type=MovieSceneByteTrack
       display_name=Visibility
       sections [{type=MovieSceneByteSection, end_frame=1, start_frame=0, channels [None {type=MovieSceneScriptingByteChannel, const [4,2]}]}]
      }
      {
       type=MovieSceneFloatTrack
       display_name=Render Opacity
       sections [{type=MovieSceneFloatSection, end_frame=0, start_frame=0, channels [None {type=MovieSceneScriptingFloatChannel, const [1,1]}]}]
      }
     ]
    }
    TextBlock_SubSplash {
     id=#6
     possessed_class=UEFN_TextBlock_C
     tracks [{
       type=MovieSceneFloatTrack
       display_name=Render Opacity
       sections [{
         type=MovieSceneFloatSection
         end_frame=43
         start_frame=0
         channels [None {type=MovieSceneScriptingFloatChannel, keys [[0,0],[20,0],[23,1],[43,1]]}]
        }]
      }]
    }
    OverlayTextWrapper {
     id=#7
     possessed_class=Overlay
     tracks [{
       type=MovieSceneFloatTrack
       display_name=Render Opacity
       sections [{
         type=MovieSceneFloatSection
         end_frame=43
         start_frame=0
         channels [None {type=MovieSceneScriptingFloatChannel, keys [[0,0],[6,0],[12,1],[43,1]]}]
        }]
      }]
    }
    Image_Banner {
     id=#8
     possessed_class=Image
     tracks [{
       type=MovieSceneWidgetMaterialTrack
       sections [{
         type=MovieSceneParameterSection
         end_frame=43
         start_frame=0
         channels [
          Lightning Border - Reveal {type=MovieSceneScriptingFloatChannel, keys [[0,0],[1,0],[12,1],[43,1]]}
          Lightning Border - Reveal Angle {type=MovieSceneScriptingFloatChannel, thinned_from=5, keys [[0,0.1],[6,0.1],[12,-0.2],[43,-0.2]]}
          Lightning Accent - Opacity {type=MovieSceneScriptingFloatChannel, keys [[0,0],[4,0],[8,0.02],[10,0.4],[15,0.02],[43,0.02]]}
          Banner - Reveal {type=MovieSceneScriptingFloatChannel, keys [[0,1],[6,1],[10,0.3],[43,0.3]]}
          Banner - Reveal Angle {type=MovieSceneScriptingFloatChannel, keys [[0,0],[6,0],[9,0.1],[43,0.1]]}
          Banner Opacity {type=MovieSceneScriptingFloatChannel, keys [[0,0],[6,0],[8,1],[43,1]]}
          Accents - ColorR {type=MovieSceneScriptingFloatChannel, const [0.8714,1]}
          Accents - ColorG {type=MovieSceneScriptingFloatChannel, const [0.0561,1]}
          Accents - ColorB {type=MovieSceneScriptingFloatChannel, const [0.1651,1]}
          Accents - ColorA {type=MovieSceneScriptingFloatChannel, const [1,1]}
          Banner - Inner ColorR {type=MovieSceneScriptingFloatChannel, const [0.1094,1]}
          Banner - Inner ColorG {type=MovieSceneScriptingFloatChannel, const [0.0568,1]}
          Banner - Inner ColorB {type=MovieSceneScriptingFloatChannel, const [0.0632,1]}
          Banner - Inner ColorA {type=MovieSceneScriptingFloatChannel, const [1,1]}
          Banner - Outer ColorR {type=MovieSceneScriptingFloatChannel, const [0.0503,1]}
          Banner - Outer ColorG {type=MovieSceneScriptingFloatChannel, const [0.0137,1]}
          Banner - Outer ColorB {type=MovieSceneScriptingFloatChannel, const [0.0166,1]}
          Banner - Outer ColorA {type=MovieSceneScriptingFloatChannel, const [1,1]}
          Lightning Accent - ColorR {type=MovieSceneScriptingFloatChannel, const [0.8714,1]}
          Lightning Accent - ColorG {type=MovieSceneScriptingFloatChannel, const [0.0561,1]}
          Lightning Accent - ColorB {type=MovieSceneScriptingFloatChannel, const [0.1651,1]}
          Lightning Accent - ColorA {type=MovieSceneScriptingFloatChannel, const [1,1]}
         ]
        }]
      }]
    }
    Image_HighThreatTextGlow {
     id=#9
     possessed_class=Image
     tracks [
      {
       type=MovieSceneFloatTrack
       display_name=Render Opacity
       sections [{type=MovieSceneFloatSection, channels [None {type=MovieSceneScriptingFloatChannel, keys [[17,0],[21.0005,0.5],[26,0.06]]}]}]
      }
      {
       type=MovieScene2DTransformTrack
       sections [{
         type=MovieScene2DTransformSection
         channels [
          Translation.X {type=MovieSceneScriptingFloatChannel, const [10,2]}
          Translation.Y {type=MovieSceneScriptingFloatChannel, const [0,1]}
          Angle {type=MovieSceneScriptingFloatChannel, const [90,1]}
          Scale.X {type=MovieSceneScriptingFloatChannel, const [0.15,1]}
          Scale.Y {type=MovieSceneScriptingFloatChannel, keys [[16.5,10],[20.5,15],[26,10]]}
          Shear.X {type=MovieSceneScriptingFloatChannel, const [0,1]}
          Shear.Y {type=MovieSceneScriptingFloatChannel, const [0,1]}
         ]
        }]
      }
     ]
    }
    Image_Icon {
     id=#10
     possessed_class=Image
     tracks [{
       type=MovieSceneWidgetMaterialTrack
       sections [{
         type=MovieSceneParameterSection
         end_frame=43
         start_frame=0
         channels [
          Accent Thickness Offset {type=MovieSceneScriptingFloatChannel, default=0}
          Accent Opacity {type=MovieSceneScriptingFloatChannel, keys [[0,0],[5,0],[9,1],[43,1]]}
          Glow Base Opacity {type=MovieSceneScriptingFloatChannel, keys [[0,0],[7,0],[11,1],[43,1]]}
          Head Scale {type=MovieSceneScriptingFloatChannel, thinned_from=5, keys [[0,0.76],[10,0.76],[11,1],[43,1]]}
          Head White Out {type=MovieSceneScriptingFloatChannel, thinned_from=6, keys [[0,0],[9,0],[10,0.4],[13,0],[43,0]]}
          Glow Strength {type=MovieSceneScriptingFloatChannel, keys [[0,0.4],[8,0.4],[13,1],[43,1]]}
          Head OffsetR {type=MovieSceneScriptingFloatChannel, default=0}
          Head OffsetG {type=MovieSceneScriptingFloatChannel, keys [[0,-0.5],[8,-0.5],[11,0.05],[13,0],[43,0]]}
          Head OffsetB {type=MovieSceneScriptingFloatChannel, default=0}
          Head OffsetA {type=MovieSceneScriptingFloatChannel, default=1}
          Glow ColorR {type=MovieSceneScriptingFloatChannel, keys [[0,1],[9,0.8976],[10,1],[13,0.8976],[43,0.8976]]}
          Glow ColorG {type=MovieSceneScriptingFloatChannel, keys [[0,1],[9,0.0497],[10,1],[13,0.0497],[43,0.0497]]}
          Glow ColorB {type=MovieSceneScriptingFloatChannel, keys [[0,1],[9,0.1151],[10,1],[13,0.1151],[43,0.1151]]}
          Glow ColorA {type=MovieSceneScriptingFloatChannel, const [1,5]}
          Glow Base ColorR {type=MovieSceneScriptingFloatChannel, keys [[0,0.8976],[9,0.8976],[10,1],[13,0.8976],[43,0.8976]]}
          Glow Base ColorG {type=MovieSceneScriptingFloatChannel, keys [[0,0.0497],[9,0.0497],[10,0.2189],[13,0.0497],[43,0.0497]]}
          Glow Base ColorB {type=MovieSceneScriptingFloatChannel, keys [[0,0.1151],[9,0.1151],[10,0.5064],[13,0.1151],[43,0.1151]]}
          Glow Base ColorA {type=MovieSceneScriptingFloatChannel, const [1,5]}
         ]
        }]
      }]
    }
    Overlay_AlertBannerWrapper {
     id=#11
     possessed_class=Overlay
     tracks [
      {
       type=MovieScene2DTransformTrack
       sections [{
         type=MovieScene2DTransformSection
         end_frame=43
         start_frame=0
         channels [
          Translation.X {type=MovieSceneScriptingFloatChannel, keys [[0,-200],[17.9545,0],[43.5,0]]}
          Translation.Y {type=MovieSceneScriptingFloatChannel, const [0,2]}
          Angle {type=MovieSceneScriptingFloatChannel, const [0,2]}
          Scale.X {type=MovieSceneScriptingFloatChannel, const [1,2]}
          Scale.Y {type=MovieSceneScriptingFloatChannel, const [1,2]}
          Shear.X {type=MovieSceneScriptingFloatChannel, const [0,2]}
          Shear.Y {type=MovieSceneScriptingFloatChannel, const [0,2]}
         ]
        }]
      }
      {
       type=MovieSceneFloatTrack
       display_name=Render Opacity
       sections [{
         type=MovieSceneFloatSection
         end_frame=43
         start_frame=0
         channels [None {type=MovieSceneScriptingFloatChannel, keys [[0,0],[35.9095,1],[43.5,1]]}]
        }]
      }
      {
       type=MovieSceneByteTrack
       display_name=Visibility
       sections [{type=MovieSceneByteSection, end_frame=1, start_frame=0, channels [None {type=MovieSceneScriptingByteChannel, keys [[0,2],[1,4]]}]}]
      }
     ]
    }
    Image_ThreatBacker {
     id=#12
     possessed_class=Image
     tracks [{
       type=MovieSceneWidgetMaterialTrack
       sections [{
         type=MovieSceneParameterSection
         end_frame=43
         start_frame=0
         channels [
          Outline Thickness {type=MovieSceneScriptingFloatChannel, const [1,1]}
          Outline Master Opacity {type=MovieSceneScriptingFloatChannel, keys [[0,0],[12.5,0],[21,1],[43.5,1]]}
          Inner Shape Master Opacity {type=MovieSceneScriptingFloatChannel, keys [[0,0],[15.5,0],[19.5,0.4],[43.5,0.4]]}
          LowHighSwitch {type=MovieSceneScriptingFloatChannel, keys [[0,0],[12,0],[19,1],[43.5,1]]}
         ]
        }]
      }]
    }
    Text_IncomingSubheader {
     id=#13
     possessed_class=CommonTextBlock
     tracks [{
       type=MovieSceneColorTrack
       display_name=Color and Opacity
       sections [{
         type=MovieSceneColorSection
         channels [
          Color.R {type=MovieSceneScriptingFloatChannel, keys [[12,1],[19,0]]}
          Color.G {type=MovieSceneScriptingFloatChannel, keys [[12,0.2346],[19,0]]}
          Color.B {type=MovieSceneScriptingFloatChannel, const [0,2]}
          Color.A {type=MovieSceneScriptingFloatChannel, const [1,2]}
         ]
        }]
      }]
    }
   ]
   display_rate {numerator=30}
  }
 }
 Hide {
  binding_count=10
  length_seconds=1.1
  object=&self:Hide
  sequence {
   bindings [
    TextBlock_SubSplash {
     id=#14
     possessed_class=UEFN_TextBlock_C
     tracks [{
       type=MovieSceneFloatTrack
       display_name=Render Opacity
       sections [{type=MovieSceneFloatSection, end_frame=0, start_frame=0, channels [None {type=MovieSceneScriptingFloatChannel, const [1,1]}]}]
      }]
    }
    Image_ThreatBacker {
     id=#15
     possessed_class=Image
     tracks [{
       type=MovieSceneWidgetMaterialTrack
       sections [{
         type=MovieSceneParameterSection
         end_frame=33
         start_frame=0
         channels [Outline Master Opacity {type=MovieSceneScriptingFloatChannel, keys [[0,1],[18,0]]}]
        }]
      }]
    }
    Image_IncomingIcon {
     id=#16
     possessed_class=Image
     tracks [{
       type=MovieSceneWidgetMaterialTrack
       sections [{
         type=MovieSceneParameterSection
         end_frame=33
         start_frame=0
         channels [
          "! Opacity" {type=MovieSceneScriptingFloatChannel, keys [[0,1],[18,0]]}
          Triangle Master Opacity {type=MovieSceneScriptingFloatChannel, keys [[0,1],[18,0]]}
          "! Translucency" {type=MovieSceneScriptingFloatChannel, keys [[0,0.8],[18,0]]}
          Gradient Opacity {type=MovieSceneScriptingFloatChannel, keys [[0,1],[18,0]]}
          Bracket Opacity {type=MovieSceneScriptingFloatChannel, keys [[0,1],[18,0]]}
         ]
        }]
      }]
    }
    Overlay_IncomingBannerWrapper {
     id=#17
     possessed_class=Overlay
     tracks [
      {
       type=MovieSceneFloatTrack
       display_name=Render Opacity
       sections [{type=MovieSceneFloatSection, end_frame=33, start_frame=0, channels [None {type=MovieSceneScriptingFloatChannel, keys [[0,1],[18,0],[33,0]]}]}]
      }
      {
       type=MovieSceneByteTrack
       display_name=Visibility
       sections [{type=MovieSceneByteSection, end_frame=33, start_frame=0, channels [None {type=MovieSceneScriptingByteChannel, keys [[0,4],[32,4],[33,2]]}]}]
      }
     ]
    }
    Overlay_SplashBannerWrapper {
     id=#18
     possessed_class=Overlay
     tracks [
      {
       type=MovieSceneFloatTrack
       display_name=Render Opacity
       sections [{type=MovieSceneFloatSection, end_frame=33, start_frame=0, channels [None {type=MovieSceneScriptingFloatChannel, keys [[0,1],[18,0],[33,0]]}]}]
      }
      {
       type=MovieSceneByteTrack
       display_name=Visibility
       sections [{type=MovieSceneByteSection, end_frame=33, start_frame=0, channels [None {type=MovieSceneScriptingByteChannel, keys [[0,4],[32,4],[33,2]]}]}]
      }
     ]
    }
    Overlay_AlertBannerWrapper {
     id=#11
     possessed_class=Overlay
     tracks [
      {
       type=MovieScene2DTransformTrack
       sections [{
         type=MovieScene2DTransformSection
         end_frame=33
         start_frame=0
         channels [
          Translation.X {type=MovieSceneScriptingFloatChannel, keys [[0,0],[17.9545,-200],[33,-200]]}
          Translation.Y {type=MovieSceneScriptingFloatChannel, const [0,1]}
          Angle {type=MovieSceneScriptingFloatChannel, const [0,1]}
          Scale.X {type=MovieSceneScriptingFloatChannel, const [1,1]}
          Scale.Y {type=MovieSceneScriptingFloatChannel, const [1,1]}
          Shear.X {type=MovieSceneScriptingFloatChannel, const [0,1]}
          Shear.Y {type=MovieSceneScriptingFloatChannel, const [0,1]}
         ]
        }]
      }
      {
       type=MovieSceneFloatTrack
       display_name=Render Opacity
       sections [{
         type=MovieSceneFloatSection
         end_frame=33
         start_frame=0
         channels [None {type=MovieSceneScriptingFloatChannel, keys [[0,1],[17.9545,0],[33,0]]}]
        }]
      }
      {
       type=MovieSceneByteTrack
       display_name=Visibility
       sections [{type=MovieSceneByteSection, end_frame=33, start_frame=0, channels [None {type=MovieSceneScriptingByteChannel, keys [[0,4],[32,4],[33,2]]}]}]
      }
     ]
    }
    Image_SubBanner {
     id=#19
     possessed_class=Image
     tracks [{
       type=MovieSceneWidgetMaterialTrack
       sections [{
         type=MovieSceneParameterSection
         end_frame=33
         start_frame=0
         channels [
          Accents - Reveal {type=MovieSceneScriptingFloatChannel, const [0.38,1]}
          Accents - Reveal Angle {type=MovieSceneScriptingFloatChannel, const [0.003,1]}
          Accents - Invert Reveal? {type=MovieSceneScriptingFloatChannel, const [1,1]}
          Banner Opacity {type=MovieSceneScriptingFloatChannel, const [0.7,1]}
          Accent ColorR {type=MovieSceneScriptingFloatChannel, const [0.0189,1]}
          Accent ColorG {type=MovieSceneScriptingFloatChannel, const [0.0498,1]}
          Accent ColorB {type=MovieSceneScriptingFloatChannel, const [0.0781,1]}
          Accent ColorA {type=MovieSceneScriptingFloatChannel, const [1,1]}
          Banner ColorR {type=MovieSceneScriptingFloatChannel, const [0.6172,1]}
          Banner ColorG {type=MovieSceneScriptingFloatChannel, const [0.8632,1]}
          Banner ColorB {type=MovieSceneScriptingFloatChannel, const [0.8879,1]}
          Banner ColorA {type=MovieSceneScriptingFloatChannel, const [1,1]}
         ]
        }]
      }]
    }
    Image_IncomingBacker {
     id=#20
     possessed_class=Image
     tracks [{
       type=MovieSceneWidgetMaterialTrack
       sections [{
         type=MovieSceneParameterSection
         end_frame=33
         start_frame=0
         channels [
          Backer_Opacity {type=MovieSceneScriptingFloatChannel, keys [[0,0.85],[17.9545,0]]}
          Dots Opacity {type=MovieSceneScriptingFloatChannel, keys [[0,1],[17.9545,0]]}
          Outline_Opacity {type=MovieSceneScriptingFloatChannel, keys [[0,0.8],[17.9545,0]]}
          Warning Opacity {type=MovieSceneScriptingFloatChannel, keys [[0,1],[17.9545,0]]}
         ]
        }]
      }]
    }
    Image_Icon {
     id=#21
     possessed_class=Image
     tracks [{
       type=MovieSceneWidgetMaterialTrack
       sections [{
         type=MovieSceneParameterSection
         end_frame=33
         start_frame=0
         channels [
          Accent Thickness Offset {type=MovieSceneScriptingFloatChannel, const [0,1]}
          Accent Opacity {type=MovieSceneScriptingFloatChannel, const [1,1]}
          Glow Base Opacity {type=MovieSceneScriptingFloatChannel, const [1,1]}
          Head Scale {type=MovieSceneScriptingFloatChannel, const [1,1]}
          Head White Out {type=MovieSceneScriptingFloatChannel, const [0,1]}
          Glow Strength {type=MovieSceneScriptingFloatChannel, const [1,1]}
          Head OffsetR {type=MovieSceneScriptingFloatChannel, const [0,1]}
          Head OffsetG {type=MovieSceneScriptingFloatChannel, const [0,1]}
          Head OffsetB {type=MovieSceneScriptingFloatChannel, const [0,1]}
          Head OffsetA {type=MovieSceneScriptingFloatChannel, const [1,1]}
          Glow ColorR {type=MovieSceneScriptingFloatChannel, const [0.8976,1]}
          Glow ColorG {type=MovieSceneScriptingFloatChannel, const [0.0497,1]}
          Glow ColorB {type=MovieSceneScriptingFloatChannel, const [0.1151,1]}
          Glow ColorA {type=MovieSceneScriptingFloatChannel, const [1,1]}
          Glow Base ColorR {type=MovieSceneScriptingFloatChannel, const [0.8976,1]}
          Glow Base ColorG {type=MovieSceneScriptingFloatChannel, const [0.0497,1]}
          Glow Base ColorB {type=MovieSceneScriptingFloatChannel, const [0.1151,1]}
          Glow Base ColorA {type=MovieSceneScriptingFloatChannel, const [1,1]}
         ]
        }]
      }]
    }
    Image_Banner {
     id=#22
     possessed_class=Image
     tracks [{
       type=MovieSceneWidgetMaterialTrack
       sections [{
         type=MovieSceneParameterSection
         end_frame=33
         start_frame=0
         channels [
          Lightning Border - Reveal {type=MovieSceneScriptingFloatChannel, const [1,1]}
          Lightning Border - Reveal Angle {type=MovieSceneScriptingFloatChannel, const [-0.02,1]}
          Lightning Accent - Opacity {type=MovieSceneScriptingFloatChannel, keys [[0,0.02],[18,0]]}
          Banner - Reveal {type=MovieSceneScriptingFloatChannel, const [0.3,1]}
          Banner - Reveal Angle {type=MovieSceneScriptingFloatChannel, const [0.1,1]}
          Banner Opacity {type=MovieSceneScriptingFloatChannel, keys [[0,1],[18,0]]}
          Accents - ColorR {type=MovieSceneScriptingFloatChannel, keys [[0,0.8714],[18,0]]}
          Accents - ColorG {type=MovieSceneScriptingFloatChannel, keys [[0,0.0561],[18,0]]}
          Accents - ColorB {type=MovieSceneScriptingFloatChannel, keys [[0,0.1651],[18,0]]}
          Accents - ColorA {type=MovieSceneScriptingFloatChannel, const [1,2]}
          Banner - Inner ColorR {type=MovieSceneScriptingFloatChannel, keys [[0,0.1094],[18,0]]}
          Banner - Inner ColorG {type=MovieSceneScriptingFloatChannel, keys [[0,0.0568],[18,0]]}
          Banner - Inner ColorB {type=MovieSceneScriptingFloatChannel, keys [[0,0.0632],[18,0]]}
          Banner - Inner ColorA {type=MovieSceneScriptingFloatChannel, keys [[0,1],[18,0]]}
          Banner - Outer ColorR {type=MovieSceneScriptingFloatChannel, keys [[0,0.0503],[18,0]]}
          Banner - Outer ColorG {type=MovieSceneScriptingFloatChannel, keys [[0,0.0137],[18,0]]}
          Banner - Outer ColorB {type=MovieSceneScriptingFloatChannel, keys [[0,0.0166],[18,0]]}
          Banner - Outer ColorA {type=MovieSceneScriptingFloatChannel, keys [[0,1],[18,0]]}
          Lightning Accent - ColorR {type=MovieSceneScriptingFloatChannel, keys [[0,0.8714],[18,0]]}
          Lightning Accent - ColorG {type=MovieSceneScriptingFloatChannel, keys [[0,0.0561],[18,0]]}
          Lightning Accent - ColorB {type=MovieSceneScriptingFloatChannel, keys [[0,0.1651],[18,0]]}
          Lightning Accent - ColorA {type=MovieSceneScriptingFloatChannel, const [1,2]}
         ]
        }]
      }]
    }
   ]
   display_rate {numerator=30}
  }
 }
]
graphs [
 EventGraph {nodes [
   HandleShowHideAnimationFinished() {is=title K2Node_CustomEvent}
   > Call On Alert Finished!(execute:Exec, self:WBP Toy Fox Alert Banner Object Reference) -> (then:Exec) {is=title K2Node_CallDelegate}
   > Set bAlertActive(bAlertActive:bool=false) -> (Output_Get:bool)
   > SetVisibility(InVisibility=Collapsed)
   HandleShowHideAnimationStarted() {is=title K2Node_CustomEvent}
   > SetVisibility(InVisibility=SelfHitTestInvisible)
   n1 = UpdateAlertBanner!() -> (OutputDelegate, then, AlertNotificationDatatableStruct:"Alert Notification Datatable Struct Structure (by ref)") {is=title K2Node_CustomEvent}
   > SetActiveWidgetIndex(
    self:Common Visibility Switcher Object Reference=Get AlertStyleSwitcher() -> (AlertStyleSwitcher:Common Visibility Switcher Object Reference)
    Index=1
   )
   n2 = Break Alert Notification Datatable Struct!(AlertNotificationDatatableStruct:"Alert Notification Datatable Struct Structure (by ref)"=@n1.AlertNotificationDatatableStruct) -> (
    AlertGameplayTag_49_#23:Gameplay Tag Structure
    Enabled_55_#24:Scalable Float Structure
    TitleText_42_#25:Text
    BodyText_44_#26:Text
    PortraitTexture_45_#27:Texture 2D Soft Object Reference
    DisplayDuration_58_#28:Scalable Float Structure
    TargetType_46_#29:AlertTargetTypeEnum Enum
    AlertVOEnabled_54_#30:Scalable Float Structure
    AlertVO_39_#31:Sound Base Soft Object Reference
    DelayBeforeVO_25_#32:double=~
    bPowerRangerFormat_27_#33:bool=~
    AlertStinger_36_#34:Sound Base Soft Object Reference
    AlertStringerEventTag_63_#35:Gameplay Tag Structure
    AlertStingerEventGroup_67_#36:Name
    DistanceFromMegazord_60_#37:Scalable Float Structure
    IsZordon?_72_#38:bool=~
   ) {is=title K2Node_BreakStruct}
   > SetZordonColors!(execute, self, IsZordon:bool=@n2.IsZordon?_72_#38) -> (then)
   > n3 = Async Load Asset(Asset=@n2.PortraitTexture_45_#27)
   > "SetText (Text)"(self=Get Text_Name() -> (Text_Name:Common Text Object Reference), InText=@n2.TitleText_42_#25)
   > SetText(self=Get Text_Body() -> (Text_Body:Common Rich Text Block Object Reference), InText=@n2.BodyText_44_#26)
   > n4 = Sequence()
   > Branch(Condition=RandomizeChanceToPlay!(self, Input:"Scalable Float Structure (by ref)"=@n2.AlertVOEnabled_54_#30) -> (ReturnValue:bool))
   > Delay(Duration=@n2.DelayBeforeVO_25_#32)
   > n5 = Async Load Asset(Asset=@n2.AlertVO_39_#31)
   n6 = Cast To SoundBase!(execute=@n5.Completed, Object=@n5) -> (then, CastFailed, AsSound Base:Sound Base Object Reference) {is=title K2Node_DynamicCast}
   > PlaySound2D(Sound=@n6)
   n7 = Async Load Asset(execute=@n4.then_1, Asset=@n2.AlertStinger_36_#34)
   n8 = Cast To SoundBase(execute=@n7.Completed, Object=@n7)
   > n9 = SpawnSound2D!(
    execute
    Sound:Sound Base Object Reference=@n8
    VolumeMultiplier:float=1
    PitchMultiplier:float=1
    StartTime:float=0
    ConcurrencySettings:Sound Concurrency Object Reference
    bPersistAcrossLevelTransition:bool=false
    bAutoDestroy:bool=true
   ) -> (then, ReturnValue:Audio Component Object Reference)
   > AddEventByComponent!(
    execute
    self:Music Event Subsystem Object Reference=Get MusicEventSubsystem() -> (ReturnValue:Music Event Subsystem Object Reference) {is=title K2Node_GetSubsystem}
    Component:Audio Component Object Reference=@n9
    EventTag:Gameplay Tag Structure=@n2.AlertStringerEventTag_63_#35
    EventBehaviorTags:Gameplay Tag Container Structure=(GameplayTags=)
    MusicClock:Music Clock Object Reference
    GroupName:Name=@n2.AlertStingerEventGroup_67_#36
   ) -> (then)
   n10 = GetDynamicMaterial(execute=@n3.Completed, self=Get Image_CharacterPortrait() -> (Image_CharacterPortrait:Image Object Reference))
   > SetTextureParameterValue(self=@n10, ParameterName=AlertPortrait, Value=Cast To Texture2D!(Object=@n3) -> (AsTexture 2D, bSuccess:bool=~))
   n11 = UpdateAlertWithRow!() -> (OutputDelegate, then, AlertTag:Gameplay Tag Structure) {is=title K2Node_CustomEvent}
   > n12 = GetAlertNotificationFromTag!(execute, self, GameplayTag:Gameplay Tag Structure=@n11.AlertTag) -> (then, bValid:bool, ReturnValue:Alert Notification Datatable Struct Structure)
   > Branch(Condition=@n12)
   > StopAllAnimations(self:User Widget Object Reference)
   > Clear and Invalidate Timer by Handle(Handle=Get PlayAnimHideTimerHandle() -> (PlayAnimHideTimerHandle:Timer Handle Structure))
   > Call On Alert Started!(execute:Exec, self:WBP Toy Fox Alert Banner Object Reference) -> (then:Exec) {is=title K2Node_CallDelegate}
   > Set bAlertActive(bAlertActive=true)
   > n13 = Sequence()
   > n14 = Play Animation with Finished event!(
    execute:Exec
    Widget:User Widget Object Reference=Self-Reference()
    InAnimation:Widget Animation Object Reference=Get Show() -> (Show:Widget Animation Object Reference)
    StartAtTime:float=0
    NumLoopsToPlay:int=1
    PlayMode:EUMGSequencePlayMode Enum
    PlaybackSpeed:float=1
   ) -> (then:Exec, Result:Widget Animation Handle Structure, Finished:Exec) {is=title K2Node_PlayAnimation2}
   n15 = PlayAnimHide() {is=title K2Node_CustomEvent}
   n16 = Set Timer by Event(
    execute=@n14.Finished
    Delegate=@n15
    Time=GetValueAtLevel!(
     Input:"Scalable Float Structure (by ref)"=Break Alert Notification Datatable Struct(AlertNotificationDatatableStruct=@n12.ReturnValue) -> (
      DelayBeforeVO_25_#32=~
      bPowerRangerFormat_27_#33=~
      IsZordon?_72_#38=~
     ).DisplayDuration_58_#28
     Level:float=0
    ) -> (ReturnValue:float)
    bLooping=false
   )
   > Set PlayAnimHideTimerHandle(PlayAnimHideTimerHandle:Timer Handle Structure=@n16) -> (Output_Get:Timer Handle Structure)
   n17 = Branch(
    execute=@n13.then_1
    Condition=MatchesAnyTags!(
     TagOne:Gameplay Tag Structure=@n11.AlertTag
     OtherContainer:"Gameplay Tag Container Structure (by ref)"=MakeLiteralGameplayTagContainer!(Value:Gameplay Tag Container Structure=(GameplayTags=((TagName="ToyFox.HUD.Alert.Megazord.Ready"),(TagName="ToyFox.HUD.Alert.Megazord.Exit"),(TagName="ToyFox.HUD.Alert.Megazord.Claimed")))) -> (ReturnValue:Gameplay Tag Container Structure)
     bExactMatch:bool=false
    ) -> (ReturnValue:bool)
   )
   > UpdateSplashBanner!(execute, self, AlertNotificationDatatableStruct:"Alert Notification Datatable Struct Structure (by ref)"=@n12.ReturnValue) -> (then)
   n18 = Branch(
    execute=@n17.else
    Condition=MatchesTag!(
     TagOne:Gameplay Tag Structure=@n11.AlertTag
     TagTwo:Gameplay Tag Structure=(TagName="ToyFox.HUD.Alert.Queen.Detected")
     bExactMatch:bool=false
    ) -> (ReturnValue:bool)
   )
   > UpdateIncomingBanner!(execute, self, AlertNotificationDatatableStruct:"Alert Notification Datatable Struct Structure (by ref)"=@n12.ReturnValue) -> (then)
   UpdateAlertBanner!(execute=@n18.else, self, AlertNotificationDatatableStruct:"Alert Notification Datatable Struct Structure (by ref)"=@n12.ReturnValue) -> (then) {is=title K2Node_CallFunction}
   n19 = HandleAlertNotificationReceived!() -> (OutputDelegate, then, AlertNotification:Alert Notification Struct Structure) {is=title K2Node_CustomEvent}
   > n20 = ShouldDisplayAlertNotification!(execute, self, Notification:"Alert Notification Struct Structure (by ref)"=@n19.AlertNotification) -> (then, bReturnValue:bool)
   > Branch(Condition=@n20)
   > UpdateAlertWithRow!(
    execute
    self
    AlertTag:Gameplay Tag Structure=Break Alert Notification Struct!(AlertNotificationStruct:"Alert Notification Struct Structure (by ref)"=@n19.AlertNotification) -> (AlertGameplayTag_12_#39:Gameplay Tag Structure) {is=title K2Node_BreakStruct}
   ) -> (then) {is=title K2Node_CallFunction}
   n21 = UpdateSplashBanner!() -> (OutputDelegate, then, AlertNotificationDatatableStruct:"Alert Notification Datatable Struct Structure (by ref)") {is=title K2Node_CustomEvent}
   > SetActiveWidgetIndex(self=Get AlertStyleSwitcher(), Index=0)
   n22 = Break Alert Notification Datatable Struct(AlertNotificationDatatableStruct=@n21.AlertNotificationDatatableStruct) -> (
    DelayBeforeVO_25_#32=~
    bPowerRangerFormat_27_#33=~
    IsZordon?_72_#38=~
   )
   > HideSubheader!(execute, self, IsHidden:bool=TextIsEmpty(InText=@n22.BodyText_44_#26)) -> (then)
   > "SetText (Text)"(self=Get Text_Splash() -> (Text_Splash:Common Text Object Reference), InText=@n22.TitleText_42_#25)
   > "SetText (Text)"(self=Get Text_SplashShadow() -> (Text_SplashShadow:Common Text Object Reference), InText=@n22.TitleText_42_#25)
   > "SetText (Text)"(
    self=Get TextBlock_SubSplash() -> (TextBlock_SubSplash:Text Block Object Reference)
    InText=@n22.BodyText_44_#26
   )
   > n23 = Async Load Asset(Asset=@n22.AlertStinger_36_#34)
   n24 = Cast To SoundBase(execute=@n23.Completed, Object=@n23)
   > n25 = SpawnSound2D(Sound=@n24, VolumeMultiplier=1, PitchMultiplier=1, StartTime=0, bPersistAcrossLevelTransition=false, bAutoDestroy=true)
   > AddEventByComponent(
    self=Get MusicEventSubsystem()
    Component=@n25
    EventTag=@n22.AlertStringerEventTag_63_#35
    EventBehaviorTags=(GameplayTags=)
    GroupName=@n22.AlertStingerEventGroup_67_#36
   )
   n26 = UpdateIncomingBanner!() -> (OutputDelegate, then, AlertNotificationDatatableStruct:"Alert Notification Datatable Struct Structure (by ref)") {is=title K2Node_CustomEvent}
   > SetActiveWidgetIndex(self=Get AlertStyleSwitcher(), Index=2)
   n27 = Break Alert Notification Datatable Struct(AlertNotificationDatatableStruct=@n26.AlertNotificationDatatableStruct) -> (
    DelayBeforeVO_25_#32=~
    bPowerRangerFormat_27_#33=~
    IsZordon?_72_#38=~
   )
   > "SetText (Text)"(
    self=Get Text_IncomingTitle() -> (Text_IncomingTitle:Common Text Object Reference)
    InText=@n27.TitleText_42_#25
   )
   > "SetText (Text)"(
    self=Get Text_IncomingSubheader() -> (Text_IncomingSubheader:Common Text Object Reference)
    InText=@n27.BodyText_44_#26
   )
   > n28 = Async Load Asset(Asset=@n27.AlertStinger_36_#34)
   n29 = Cast To SoundBase(execute=@n28.Completed, Object=@n28)
   > n30 = SpawnSound2D(Sound=@n29, VolumeMultiplier=1, PitchMultiplier=1, StartTime=0, bPersistAcrossLevelTransition=false, bAutoDestroy=true)
   > AddEventByComponent(
    self=Get MusicEventSubsystem()
    Component=@n30
    EventTag=@n27.AlertStringerEventTag_63_#35
    EventBehaviorTags=(GameplayTags=)
    GroupName=@n27.AlertStingerEventGroup_67_#36
   )
   Event OnInitialized()
   > BindToAnimationFinished(Animation=Get Hide() -> (Hide:Widget Animation Object Reference), Delegate=Create Event())
   > BindToAnimationStarted!(execute, self:User Widget Object Reference, Animation:Widget Animation Object Reference=Get Show(), Delegate:Delegate=Create Event()) -> (then)
   Event Construct()
   > Parent: Construct()
   > SetVisibility(InVisibility=Collapsed)
   > n31 = GetGameStateAthena!(execute) -> (then, ReturnValue:Fort Game State Athena Object Reference)
   > Is Valid(InputObject=@n31)
   > n32 = GetComponentByClass!(
    execute
    self:Actor Object Reference=@n31
    ComponentClass:Actor Component Class Reference=/ToyFoxGameplay/Gameplay/B_ToyFoxGameStateComponent.B_ToyFoxGameStateComponent_C
   ) -> (then, ReturnValue:B Toy Fox Game State Component Object Reference)
   > n33 = Set GameStateComponent(GameStateComponent:B Toy Fox Game State Component Object Reference=@n32) -> (Output_Get:B Toy Fox Game State Component Object Reference)
   > Is Valid(InputObject=@n33)
   > Bind Event to On Alert Received(self:B Toy Fox Game State Component Object Reference=@n33, Delegate=Create Event()) {is=title K2Node_AddDelegate}
   Play Animation with Finished event(execute=@n15.then, Widget=Self-Reference(), InAnimation=Get Hide(), StartAtTime=0, NumLoopsToPlay=1, PlaybackSpeed=1)
  ]}
 GetAlertNotificationFromTag {nodes [
   n1 = GetAlertNotificationFromTag!() -> (then, GameplayTag:Gameplay Tag Structure) {is=title K2Node_FunctionEntry}
   > n2 = Sequence()
   n3 = Get AlertDatatable() -> (AlertDatatable:Data Table Object Reference)
   > n4 = GetDataTableRowNames(Table=@n3)
   > n5 = For Each Loop(Array:Array of Names=@n4) -> (Array Element:Name)
   > n6 = Get Data Table Row!(execute:Exec, DataTable:Data Table Object Reference=@n3, RowName:Name=@n5) -> (then:Exec, RowNotFound:Exec, ReturnValue:Alert Notification Datatable Struct Structure)
   > Branch(Condition="Equal (GameplayTag)"(
     A=Break Alert Notification Datatable Struct(AlertNotificationDatatableStruct=@n6) -> (
      DelayBeforeVO_25_#32=~
      bPowerRangerFormat_27_#33=~
      IsZordon?_72_#38=~
     )
     B=@n1
    ))
   > Return Node!(execute, bValid:bool=true, ReturnValue:Alert Notification Datatable Struct Structure=@n6)
   Return Node(execute=@n2.then_1, bValid=false)
  ]}
 HideSubheader {nodes [
   n1 = HideSubheader!() -> (then, IsHidden:bool=~) {is=title K2Node_FunctionEntry}
   > n2 = Branch(Condition=@n1)
   n3 = Get OverlaySubBanner() -> (OverlaySubBanner:Overlay Object Reference)
   > n4 = SetVisibility(self=@n3, InVisibility=Hidden)
   Return Node!(execute=[@n5.then, @n4.then])
   n5 = SetVisibility(execute=@n2.else, self=@n3, InVisibility=HitTestInvisible)
  ]}
 IsAlertActive {nodes [
   IsAlertActive() {is=title K2Node_FunctionEntry}
   > Return Node!(execute, bAlertActive:bool=Get bAlertActive() -> (bAlertActive:bool))
  ]}
 IsNonPilotPlayer {nodes [
   n1 = IsNonPilotPlayer!() -> (then, PlayerPawn:Fort Player Pawn Object Reference) {is=title K2Node_FunctionEntry}
   > n2 = Sequence()
   > Is Valid(InputObject=@n1)
   > n3 = Get!(execute:Exec) -> (then:Exec, else:Exec, GameStateComponent:B Toy Fox Game State Component Object Reference)
   n4 = Contains Item(
    TargetArray:Array of Actor Object References=GetPilotActors!(self:B Toy Fox Game State Component Object Reference=@n3) -> (PilotActors:Array of Actor Object References)
    ItemToFind:"Actor Object Reference (by ref)"=@n1
   )
   n5 = "Equal (Byte)"(
    A:Byte=GetActorTeam!(Actor:Actor Object Reference=GetGiantPossessingPlayerState!(self:B Toy Fox Game State Component Object Reference=@n3) -> (PlayerState:Fort Player State Athena Object Reference)) -> (ReturnValue:Byte=0)
    B:Byte=GetActorTeam(Actor=@n1) -> (ReturnValue=0)
   ) -> (ReturnValue:bool=~) {is=title K2Node_PromotableOperator}
   n6 = HasAnyMatchingGameplayTags!(
    self:Gameplay Tag Asset Interface Interface=@n1
    TagContainer:"Gameplay Tag Container Structure (by ref)"=MakeLiteralGameplayTagContainer(Value=(GameplayTags=((TagName="ToyFox.Player.Pilot.Primary"),(TagName="ToyFox.Player.Pilot.Secondary"),(TagName="GiantPawn"))))
   ) -> (ReturnValue:bool)
   > Return Node!(execute, ReturnValue:bool=AND Boolean!(A=NOT Boolean(A=@n5), B=NOT Boolean(A=@n4), C:bool=NOT Boolean(A=@n6)) -> (ReturnValue))
   Return Node(execute=@n2.then_1, ReturnValue=false)
  ]}
 IsPilotPlayer {nodes [
   n1 = IsPilotPlayer!() -> (then, PlayerPawn:Fort Player Pawn Object Reference) {is=title K2Node_FunctionEntry}
   > n2 = Sequence()
   > Is Valid(InputObject=@n1)
   > n3 = Get()
   n4 = GetPilotActors(self=@n3)
   n5 = MakeLiteralGameplayTagContainer(Value=(GameplayTags=((TagName="ToyFox.Player.Pilot.Primary"),(TagName="ToyFox.Player.Pilot.Secondary"),(TagName="GiantPawn"))))
   > Return Node(ReturnValue=OR Boolean(A=Contains Item(TargetArray=@n4, ItemToFind=@n1), B=HasAnyMatchingGameplayTags(self=@n1, TagContainer=@n5)))
   Return Node(execute=@n2.then_1, ReturnValue=false)
  ]}
 OnAlertFinished {nodes [OnAlertFinished() {is=title K2Node_FunctionEntry}]}
 OnAlertStarted {nodes [OnAlertStarted() {is=title K2Node_FunctionEntry}]}
 RandomizeChanceToPlay {nodes [
   n1 = RandomizeChanceToPlay!() -> (then, Input:"Scalable Float Structure (by ref)") {is=title K2Node_FunctionEntry}
   > Return Node(ReturnValue="float >= float"(A=GetValueAtLevel(Input=@n1, Level=0), B=RandomFloatInRange(Min=0.01, Max=1)))
  ]}
 SetZordonColors {nodes [
   n1 = SetZordonColors!() -> (then, IsZordon:bool=~) {is=title K2Node_FunctionEntry}
   > n2 = GetDynamicMaterial(self=Get Image_CharacterPortrait())
   > n3 = GetDynamicMaterial(self=Get Image_AlertTexbox() -> (Image_AlertTexbox:Image Object Reference))
   > n4 = GetDynamicMaterial(self=Get Image_AlertTitlebox() -> (Image_AlertTitlebox:Image Object Reference))
   > n5 = Branch(Condition=@n1)
   n6 = Get CharacterColor() -> (CharacterColor:Name=None)
   n7 = Get ZordonColor() -> (ZordonColor:Linear Color Structure)
   > SetVectorParameterValue(self=@n2, ParameterName=@n6, Value=@n7)
   > SetVectorParameterValue(self=@n3, ParameterName=@n6, Value=@n7)
   > n8 = SetVectorParameterValue(self=@n4, ParameterName=@n6, Value=@n7)
   Return Node!(execute=[@n8.then, @n11.then])
   n9 = Get RitaColor() -> (RitaColor:Linear Color Structure)
   n10 = Get CharacterColor() -> (CharacterColor=None)
   SetVectorParameterValue(execute=@n5.else, self=@n2, ParameterName=@n10, Value=@n9)
   > SetVectorParameterValue(self=@n3, ParameterName=@n10, Value=@n9)
   > n11 = SetVectorParameterValue(self=@n4, ParameterName=@n10, Value=@n9)
  ]}
 ShouldDisplayAlertNotification {nodes [
   n1 = ShouldDisplayAlertNotification!() -> (then, Notification:"Alert Notification Struct Structure (by ref)") {is=title K2Node_FunctionEntry}
   > n2 = Sequence()
   n3 = GetOwningPlayer()
   > Is Valid(InputObject=@n3)
   > n4 = Cast To FortPlayerPawn(Object=GetViewTarget!(self:Controller Object Reference=@n3) -> (ReturnValue:Actor Object Reference))
   > Set ViewTargetPlayerPawn(ViewTargetPlayerPawn:Fort Player Pawn Object Reference=@n4) -> (Output_Get:Fort Player Pawn Object Reference)
   n5 = Break Alert Notification Struct!(AlertNotificationStruct=@n1) -> (
    AlertGameplayTag_12_#39
    ExplicitPlayers_6_#40:Array of Fort Player State Athena Object References
   )
   > n6 = GetAlertNotificationFromTag(GameplayTag=@n5) {is=title K2Node_CallFunction}
   > Branch(Condition=@n6)
   n7 = Break Alert Notification Datatable Struct(AlertNotificationDatatableStruct=@n6.ReturnValue) -> (
    DelayBeforeVO_25_#32=~
    bPowerRangerFormat_27_#33=~
    IsZordon?_72_#38=~
   )
   > Branch(Condition=GetValueAsBool!(Input:"Scalable Float Structure (by ref)"=@n7.Enabled_55_#24, Level:float=0) -> (ReturnValue:bool))
   > n8 = Sequence()
   > n9 = Set DistanceFromMegazord(DistanceFromMegazord:float=GetValueAtLevel(Input=@n7.DistanceFromMegazord_60_#37, Level=0)) -> (Output_Get:float)
   > n10 = Branch(Condition="float > float"(A=@n9))
   > n11 = Get()
   n12 = GetGiantPawn!(self:B Toy Fox Game State Component Object Reference=@n11) -> (Pawn:Toy Fox Giant Player Pawn Object Reference)
   > n13 = Is Valid(InputObject=@n12)
   Set bPassDistanceTestFromMegazord(execute=[@n13.Is Not Valid, @n11.else, @n10.else], bPassDistanceTestFromMegazord:bool=true) -> (Output_Get:bool)
   n14 = Get ViewTargetPlayerPawn() -> (ViewTargetPlayerPawn:Fort Player Pawn Object Reference)
   n15 = Get DistanceFromMegazord() -> (DistanceFromMegazord:float)
   Set bPassDistanceTestFromMegazord(
    execute=@n13.Is Valid
    bPassDistanceTestFromMegazord="float < float"(
     A=GetSquaredHorizontalDistanceTo!(self:Actor Object Reference=@n14, OtherActor:Actor Object Reference=@n12) -> (ReturnValue:float)
     B=Square!(A:double=@n15) -> (ReturnValue:double)
    )
   )
   Branch(execute=@n8.then_1, Condition=Get bPassDistanceTestFromMegazord() -> (bPassDistanceTestFromMegazord:bool))
   > n16 = Switch on AlertTargetTypeEnum!(execute, Selection:AlertTargetTypeEnum Enum=@n7.TargetType_46_#29) -> (NewEnumerator0:Exec, NewEnumerator2:Exec, NewEnumerator3:Exec, NewEnumerator1:Exec) {is=title K2Node_SwitchEnum}
   n17 = AttemptGetPlayerStateForPawn!(execute=@n16.NewEnumerator1, FortPawn:Fort Player Pawn Object Reference=Get ViewTargetPlayerPawn()) -> (then, ReturnValue:Fort Player State Object Reference)
   > n18 = Cast To FortPlayerStateAthena!(execute, Object=@n17) -> (then, CastFailed, AsFort Player State Athena:Fort Player State Athena Object Reference) {is=title K2Node_DynamicCast}
   > Return Node!(
    execute
    bReturnValue:bool=Contains Item(
     TargetArray:Array of Fort Player State Athena Object References=@n5.ExplicitPlayers_6_#40
     ItemToFind:"Fort Player State Athena Object Reference (by ref)"=@n18
    )
   )
   Return Node(execute=@n16.NewEnumerator3, bReturnValue=true)
   n19 = Get ViewTargetPlayerPawn()
   Return Node(
    execute=@n16.NewEnumerator0
    bReturnValue=IsPilotPlayer!(self, PlayerPawn:Fort Player Pawn Object Reference=@n19) -> (ReturnValue:bool) {is=title K2Node_CallFunction}
   )
   Return Node(
    execute=@n16.NewEnumerator2
    bReturnValue=IsNonPilotPlayer!(self, PlayerPawn:Fort Player Pawn Object Reference=@n19) -> (ReturnValue:bool) {is=title K2Node_CallFunction}
   )
   Return Node(execute=@n2.then_1, bReturnValue=false)
  ]}
]
variables [
 AlertDatatable {type=DataTable, default=/ToyFoxGameplay/UI/DT_AlertNotificationsDatatable}
 bAlertActive {type=boolean}
 CharacterColor {type=string, default=CharacterColor}
 GameStateComponent {type=B_ToyFoxGameStateComponent_C, default=null}
 OnAlertFinished {}
 OnAlertStarted {}
 PlayAnimHideTimerHandle {type=TimerHandle}
 RitaColor {type=LinearColor, default=<rgba 1 0.162 0.552 1>}
 ZordonColor {type=LinearColor, default=<rgba 0.0331 0.6172 1 1>}
]
widgets {
 info {inheritedWidgetCount=1, parentClass=&FortHUDElementWidget, rootWidgetClass=&Overlay, widgetCount=46}
 widgets [
  Overlay_1 {
   widgetClassPath=&Overlay
   widget=Overlay_1
   properties {renderTransform {angle=0, scale=<0.8 0.8>, shear=<0 0>, translation=<0 0>}}
  }
  AlertStyleSwitcher {
   widgetClassPath=&CommonVisibilitySwitcher
   bIsVariable=true
   parent=Overlay_1
   widget=AlertStyleSwitcher
   slot {class=OverlaySlot, properties {horizontalAlignment=HAlign_Center, verticalAlignment=VAlign_Center}}
  }
  Overlay_SplashBannerWrapper {
   widgetClassPath=&Overlay
   parent=AlertStyleSwitcher
   widget=Overlay_SplashBannerWrapper
   properties {renderTransform {angle=0, scale=<0.8 0.8>, shear=<0 0>, translation=<0 0>}}
   slot {class=CommonVisibilitySwitcherSlot}
  }
  Image_Icon {
   widgetClassPath=&Image
   bIsVariable=true
   parent=Overlay_SplashBannerWrapper
   widget=Image_Icon
   properties {
    brush {
     bIsDynamicallyLoaded=false
     drawAs=Image
     imageType=FullColor
     mirroring=NoMirror
     resourceName=None
     resourceObject=/ToyFoxGameplay/UI/Materials/M_ToyFox_AlertSplash_Icon
     tiling=NoTile
     imageSize=<200 200>
     margin=<ltrb 0 0 0 0>
     outlineSettings {
      bUseBrushTransparency=false
      roundingType=HalfHeightRadius
      width=0
      color {colorUseRule=UseColor_Specified, specifiedColor=<rgba 0 0 0 0>}
      cornerRadii=<0 0 0 0>
     }
     tintColor {colorUseRule=UseColor_Specified, specifiedColor=<rgba 1 1 1 1>}
     uVRegion {bIsValid=false, max=<0 0>, min=<0 0>}
    }
    renderTransform {angle=0, scale=<1 1>, shear=<0 0>, translation=<0 -75>}
   }
   slot {class=OverlaySlot, properties {horizontalAlignment=HAlign_Center}}
  }
  CommonVisualAttachment_0 {
   widgetClassPath=&CommonVisualAttachment
   parent=Overlay_SplashBannerWrapper
   widget=CommonVisualAttachment_0
   properties {contentAnchor=<0.5 0.5>}
   slot {class=OverlaySlot, properties {horizontalAlignment=HAlign_Center, verticalAlignment=VAlign_Center}}
  }
  Overlay_StretchWrapper {
   widgetClassPath=&Overlay
   parent=CommonVisualAttachment_0
   widget=Overlay_StretchWrapper
   slot {class=SizeBoxSlot, properties {horizontalAlignment=HAlign_Center, verticalAlignment=VAlign_Center}}
  }
  Image_Banner {
   widgetClassPath=&Image
   bIsVariable=true
   parent=Overlay_StretchWrapper
   widget=Image_Banner
   properties {brush {
     bIsDynamicallyLoaded=false
     drawAs=Image
     imageType=FullColor
     mirroring=NoMirror
     resourceName=None
     resourceObject=/ToyFoxGameplay/UI/Materials/M_ToyFox_AlertSplash
     tiling=NoTile
     imageSize=<700 700>
     margin=<ltrb 0 0 0 0>
     outlineSettings {
      bUseBrushTransparency=false
      roundingType=HalfHeightRadius
      width=0
      color {colorUseRule=UseColor_Specified, specifiedColor=<rgba 0 0 0 0>}
      cornerRadii=<0 0 0 0>
     }
     tintColor {colorUseRule=UseColor_Specified, specifiedColor=<rgba 1 1 1 1>}
     uVRegion {bIsValid=false, max=<0 0>, min=<0 0>}
    }}
   slot {class=OverlaySlot, properties {horizontalAlignment=HAlign_Fill}}
  }
  OverlayTextWrapper {
   widgetClassPath=&Overlay
   parent=Overlay_StretchWrapper
   widget=OverlayTextWrapper
   slot {class=OverlaySlot, properties {horizontalAlignment=HAlign_Center, verticalAlignment=VAlign_Center}}
  }
  Text_SplashShadow {
   widgetClassPath=&CommonTextBlock
   bIsVariable=true
   parent=OverlayTextWrapper
   widget=Text_SplashShadow
   properties {
    text="{MEGAZORD READY}"
    colorAndOpacity {colorUseRule=UseColor_Specified, specifiedColor=<rgba 0.283 0.0106 0.0374 1>}
    font {
     bForceMonospaced=false
     bMaterialIsStencil=false
     fontMaterial=null
     fontObject=&Burbank
     letterSpacing=50
     monospacedWidth=1
     size=42
     skewAmount=0
     typefaceFontName=BurbankRegular-Black
     outlineSettings {
      bApplyOutlineToDropShadows=false
      bMiteredCorners=false
      bSeparateFillAlpha=false
      outlineMaterial=null
      outlineSize=0
      outlineColor=<rgba 0 0 0 1>
     }
    }
    shadowOffset=<0 0>
   }
   slot {class=OverlaySlot, properties {padding=<ltrb 120 8 120 0>}}
  }
  Text_Splash {
   widgetClassPath=&CommonTextBlock
   bIsVariable=true
   parent=OverlayTextWrapper
   widget=Text_Splash
   properties {
    text="{MEGAZORD READY}"
    font {
     bForceMonospaced=false
     bMaterialIsStencil=false
     fontMaterial=null
     fontObject=&Burbank
     letterSpacing=50
     monospacedWidth=1
     size=42
     skewAmount=0
     typefaceFontName=BurbankRegular-Black
     outlineSettings {
      bApplyOutlineToDropShadows=false
      bMiteredCorners=false
      bSeparateFillAlpha=false
      outlineMaterial=null
      outlineSize=0
      outlineColor=<rgba 0 0 0 1>
     }
    }
    shadowOffset=<0 0>
   }
   slot {class=OverlaySlot, properties {padding=<ltrb 120 0 120 0>}}
  }
  OverlaySubBanner {
   widgetClassPath=&Overlay
   bIsVariable=true
   parent=Overlay_SplashBannerWrapper
   widget=OverlaySubBanner
   properties {renderTransform {angle=0, scale=<1 1>, shear=<0 0>, translation=<0 80>}}
   slot {class=OverlaySlot, properties {horizontalAlignment=HAlign_Fill, verticalAlignment=VAlign_Center}}
  }
  Image_SubBanner {
   widgetClassPath=&Image
   bIsVariable=true
   parent=OverlaySubBanner
   widget=Image_SubBanner
   properties {brush {
     bIsDynamicallyLoaded=false
     drawAs=Image
     imageType=FullColor
     mirroring=NoMirror
     resourceName=None
     resourceObject=/ToyFoxGameplay/UI/Materials/M_ToyFox_AlertSplash_Sub
     tiling=NoTile
     imageSize=<700 50>
     margin=<ltrb 0 0 0 0>
     outlineSettings {
      bUseBrushTransparency=false
      roundingType=HalfHeightRadius
      width=0
      color {colorUseRule=UseColor_Specified, specifiedColor=<rgba 0 0 0 0>}
      cornerRadii=<0 0 0 0>
     }
     tintColor {colorUseRule=UseColor_Specified, specifiedColor=<rgba 1 1 1 1>}
     uVRegion {bIsValid=false, max=<0 0>, min=<0 0>}
    }}
   slot {class=OverlaySlot, properties {horizontalAlignment=HAlign_Center, verticalAlignment=VAlign_Bottom}}
  }
  TextBlock_SubSplash {
   widgetClassPath=&UEFN_TextBlock
   bIsVariable=true
   parent=OverlaySubBanner
   widget=TextBlock_SubSplash
   properties {
    text="{TALK TO ZORDON}"
    font {
     bForceMonospaced=false
     bMaterialIsStencil=false
     fontMaterial=null
     fontObject=&Burbank
     letterSpacing=0
     monospacedWidth=1
     size=21
     skewAmount=0
     typefaceFontName=Burbank-PlayerPower
     outlineSettings {
      bApplyOutlineToDropShadows=false
      bMiteredCorners=false
      bSeparateFillAlpha=false
      outlineMaterial=/Game/UI/Foundation/Materials/M-FontShadow
      outlineSize=3
      outlineColor=<rgba 0 0 0 0.5>
     }
    }
    strikeBrush {
     bIsDynamicallyLoaded=false
     drawAs=Image
     imageType=FullColor
     mirroring=NoMirror
     resourceName=None
     resourceObject=null
     tiling=NoTile
     imageSize=<32 32>
     margin=<ltrb 0 0 0 0>
     outlineSettings {
      bUseBrushTransparency=false
      roundingType=HalfHeightRadius
      width=0
      color {colorUseRule=UseColor_Specified, specifiedColor=<rgba 0 0 0 0>}
      cornerRadii=<0 0 0 0>
     }
     tintColor {colorUseRule=UseColor_Specified, specifiedColor=<rgba 1 1 1 1>}
     uVRegion {bIsValid=false, max=<0 0>, min=<0 0>}
    }
   }
   slot {class=OverlaySlot, properties {horizontalAlignment=HAlign_Center, verticalAlignment=VAlign_Center, padding=<ltrb 0 4 0 0>}}
  }
  Overlay_AlertBannerWrapper {
   widgetClassPath=&Overlay
   parent=AlertStyleSwitcher
   widget=Overlay_AlertBannerWrapper
   slot {class=CommonVisibilitySwitcherSlot, properties {horizontalAlignment=HAlign_Left, verticalAlignment=VAlign_Top}}
  }
  SizeBox_OverallSizeWrapper {
   widgetClassPath=&SizeBox
   parent=Overlay_AlertBannerWrapper
   widget=SizeBox_OverallSizeWrapper
   properties {bOverride_HeightOverride=true, bOverride_WidthOverride=true, heightOverride=200, widthOverride=600}
   slot {class=OverlaySlot, properties {horizontalAlignment=HAlign_Fill, verticalAlignment=VAlign_Fill}}
  }
  HorizontalBox_266 {
   widgetClassPath=&HorizontalBox
   parent=Overlay_AlertBannerWrapper
   widget=HorizontalBox_266
   properties {renderTransformPivot=<0 0>}
   slot {class=OverlaySlot, properties {horizontalAlignment=HAlign_Fill}}
  }
  Overlay_LeftSideCharacterPortrait {
   widgetClassPath=&Overlay
   parent=HorizontalBox_266
   widget=Overlay_LeftSideCharacterPortrait
   slot {class=HorizontalBoxSlot, properties {horizontalAlignment=HAlign_Left, verticalAlignment=VAlign_Top}}
  }
  Image_CharacterPortrait {
   widgetClassPath=&Image
   bIsVariable=true
   parent=Overlay_LeftSideCharacterPortrait
   widget=Image_CharacterPortrait
   properties {
    brush {
     bIsDynamicallyLoaded=false
     drawAs=Image
     imageType=FullColor
     mirroring=NoMirror
     resourceName=None
     resourceObject=/ToyFoxGameplay/UI/Materials/M_AlertCharacter
     tiling=NoTile
     imageSize=<180 180>
     margin=<ltrb 0 0 0 0>
     outlineSettings {
      bUseBrushTransparency=false
      roundingType=HalfHeightRadius
      width=0
      color {colorUseRule=UseColor_Specified, specifiedColor=<rgba 0 0 0 0>}
      cornerRadii=<0 0 0 0>
     }
     tintColor {colorUseRule=UseColor_Specified, specifiedColor=<rgba 1 1 1 1>}
     uVRegion {bIsValid=false, max=<0 0>, min=<0 0>}
    }
    renderTransform {angle=0, scale=<1 1>, shear=<2 0>, translation=<0 0>}
   }
   slot {class=OverlaySlot}
  }
  VerticalBox_RightSideTextBlocks {
   widgetClassPath=&VerticalBox
   parent=HorizontalBox_266
   widget=VerticalBox_RightSideTextBlocks
   slot {class=HorizontalBoxSlot, properties {verticalAlignment=VAlign_Top, padding=<ltrb -40 15 55 0>, size {sizeRule=Fill, value=1}}}
  }
  HorizontalBox_0 {
   widgetClassPath=&HorizontalBox
   parent=VerticalBox_RightSideTextBlocks
   widget=HorizontalBox_0
   slot {class=VerticalBoxSlot, properties {padding=<ltrb 0 0 -20 0>}}
  }
  Overlay_4 {
   widgetClassPath=&Overlay
   parent=HorizontalBox_0
   widget=Overlay_4
   slot {class=HorizontalBoxSlot, properties {horizontalAlignment=HAlign_Left, verticalAlignment=VAlign_Top}}
  }
  Image_AlertTitlebox {
   widgetClassPath=&Image
   bIsVariable=true
   parent=Overlay_4
   widget=Image_AlertTitlebox
   properties {
    brush {
     bIsDynamicallyLoaded=false
     drawAs=Image
     imageType=FullColor
     mirroring=NoMirror
     resourceName=None
     resourceObject=/ToyFoxGameplay/UI/Materials/M_ToyFox_AlertHeaderBox
     tiling=NoTile
     imageSize=<32 32>
     margin=<ltrb 0 0 0 0>
     outlineSettings {
      bUseBrushTransparency=false
      roundingType=HalfHeightRadius
      width=0
      color {colorUseRule=UseColor_Specified, specifiedColor=<rgba 0 0 0 0>}
      cornerRadii=<0 0 0 0>
     }
     tintColor {colorUseRule=UseColor_Specified, specifiedColor=<rgba 1 1 1 1>}
     uVRegion {bIsValid=false, max=<0 0>, min=<0 0>}
    }
    renderTransform {angle=0, scale=<1 2>, shear=<-1 0>, translation=<0 0>}
   }
   slot {class=OverlaySlot, properties {horizontalAlignment=HAlign_Fill, verticalAlignment=VAlign_Fill, padding=<ltrb 0 0 -30 0>}}
  }
  Text_Name {
   widgetClassPath=&CommonTextBlock
   bIsVariable=true
   parent=Overlay_4
   widget=Text_Name
   properties {
    text="{FirstName Lastname}"
    font {
     bForceMonospaced=false
     bMaterialIsStencil=false
     fontMaterial=null
     fontObject=&BurbankSmall
     letterSpacing=0
     monospacedWidth=1
     size=19.5
     skewAmount=0
     typefaceFontName=Black
     outlineSettings {
      bApplyOutlineToDropShadows=false
      bMiteredCorners=false
      bSeparateFillAlpha=false
      outlineMaterial=null
      outlineSize=0
      outlineColor=<rgba 0 0 0 1>
     }
    }
    shadowOffset=<0 0>
   }
   slot {class=OverlaySlot, properties {horizontalAlignment=HAlign_Fill, verticalAlignment=VAlign_Fill, padding=<ltrb 20 0 0 0>}}
  }
  Image_Accent {
   widgetClassPath=&Image
   bIsVariable=true
   parent=HorizontalBox_0
   widget=Image_Accent
   properties {
    brush {
     bIsDynamicallyLoaded=false
     drawAs=Image
     imageType=FullColor
     mirroring=NoMirror
     resourceName=None
     resourceObject=/ToyFoxGameplay/UI/Materials/M_ToyFox_CTAAccent
     tiling=NoTile
     imageSize=<80 50>
     margin=<ltrb 0 0 0 0>
     outlineSettings {
      bUseBrushTransparency=false
      roundingType=HalfHeightRadius
      width=0
      color {colorUseRule=UseColor_Specified, specifiedColor=<rgba 0 0 0 0>}
      cornerRadii=<0 0 0 0>
     }
     tintColor {colorUseRule=UseColor_Specified, specifiedColor=<rgba 1 1 1 1>}
     uVRegion {bIsValid=false, max=<0 0>, min=<0 0>}
    }
    renderTransform {angle=0, scale=<1 1>, shear=<0 0>, translation=<0 1.5>}
   }
   slot {class=HorizontalBoxSlot, properties {horizontalAlignment=HAlign_Left, verticalAlignment=VAlign_Top, padding=<ltrb 30 0 0 -50>}}
  }
  Overlay_3 {
   widgetClassPath=&Overlay
   parent=VerticalBox_RightSideTextBlocks
   widget=Overlay_3
   slot {class=VerticalBoxSlot, properties {verticalAlignment=VAlign_Center, padding=<ltrb 13 0 0 0>}}
  }
  Image_AlertTexbox {
   widgetClassPath=&Image
   bIsVariable=true
   parent=Overlay_3
   widget=Image_AlertTexbox
   properties {
    brush {
     bIsDynamicallyLoaded=false
     drawAs=Image
     imageType=FullColor
     mirroring=NoMirror
     resourceName=None
     resourceObject=/ToyFoxGameplay/UI/Materials/M_AlertTextbox
     tiling=NoTile
     imageSize=<0 0>
     margin=<ltrb 0 0 0 0>
     outlineSettings {
      bUseBrushTransparency=false
      roundingType=HalfHeightRadius
      width=0
      color {colorUseRule=UseColor_Specified, specifiedColor=<rgba 0 0 0 0>}
      cornerRadii=<0 0 0 0>
     }
     tintColor {colorUseRule=UseColor_Specified, specifiedColor=<rgba 1 1 1 1>}
     uVRegion {bIsValid=false, max=<0 0>, min=<0 0>}
    }
    renderTransform {angle=0, scale=<1.05 1.05>, shear=<7 0>, translation=<0 0>}
    renderTransformPivot=<0 0>
   }
   slot {class=OverlaySlot, properties {horizontalAlignment=HAlign_Fill, verticalAlignment=VAlign_Fill, padding=<ltrb 0 5 0 -5>}}
  }
  SizeBox_97 {
   widgetClassPath=&SizeBox
   parent=Overlay_3
   widget=SizeBox_97
   properties {bOverride_HeightOverride=true, bOverride_WidthOverride=true, heightOverride=80, widthOverride=400}
   slot {class=OverlaySlot, properties {horizontalAlignment=HAlign_Fill, verticalAlignment=VAlign_Fill, padding=<ltrb 30 15 0 0>}}
  }
  Text_Body {
   widgetClassPath=&CommonRichTextBlock
   bIsVariable=true
   parent=SizeBox_97
   widget=Text_Body
   properties {
    autoWrapText=true
    bIsScrollingEnabled=false
    bOverrideDefaultStyle=true
    clipping=OnDemand
    defaultTextStyleOverrideClass=/FortUILibrary/Styles/Typography/_TextStyle_Body2._TextStyle_Body2_C
    scrollOrientation=Orient_Vertical
    text="{Body text!}"
    defaultTextStyleOverride {
     overflowPolicy=Clip
     transformPolicy=None
     colorAndOpacity {colorUseRule=UseColor_Specified, specifiedColor=<rgba 1 1 1 1>}
     font {
      bForceMonospaced=false
      bMaterialIsStencil=false
      fontMaterial=null
      fontObject=&NotoSans
      letterSpacing=0
      monospacedWidth=1
      size=24
      skewAmount=0
      typefaceFontName=NotoSans-Regular
      outlineSettings {
       bApplyOutlineToDropShadows=false
       bMiteredCorners=false
       bSeparateFillAlpha=false
       outlineMaterial=null
       outlineSize=0
       outlineColor=<rgba 0 0 0 1>
      }
     }
     highlightColor {colorUseRule=UseColor_Specified, specifiedColor=<rgba 0 0 0 1>}
     highlightShape {
      bIsDynamicallyLoaded=false
      drawAs=Image
      imageType=NoImage
      mirroring=NoMirror
      resourceName=None
      resourceObject=null
      tiling=NoTile
      imageSize=<32 32>
      margin=<ltrb 0 0 0 0>
      outlineSettings {
       bUseBrushTransparency=false
       roundingType=HalfHeightRadius
       width=0
       color {colorUseRule=UseColor_Specified, specifiedColor=<rgba 0 0 0 0>}
       cornerRadii=<0 0 0 0>
      }
      tintColor {colorUseRule=UseColor_Specified, specifiedColor=<rgba 1 1 1 1>}
      uVRegion {bIsValid=false, max=<0 0>, min=<0 0>}
     }
     selectedBackgroundColor {colorUseRule=UseColor_Foreground, specifiedColor=<rgba 1 0 1 1>}
     shadowColorAndOpacity=<rgba 0 0 0 1>
     shadowOffset=<0 0>
     strikeBrush {
      bIsDynamicallyLoaded=false
      drawAs=Image
      imageType=NoImage
      mirroring=NoMirror
      resourceName=None
      resourceObject=null
      tiling=NoTile
      imageSize=<32 32>
      margin=<ltrb 0 0 0 0>
      outlineSettings {
       bUseBrushTransparency=false
       roundingType=HalfHeightRadius
       width=0
       color {colorUseRule=UseColor_Specified, specifiedColor=<rgba 0 0 0 0>}
       cornerRadii=<0 0 0 0>
      }
      tintColor {colorUseRule=UseColor_Specified, specifiedColor=<rgba 1 1 1 1>}
      uVRegion {bIsValid=false, max=<0 0>, min=<0 0>}
     }
     underlineBrush {
      bIsDynamicallyLoaded=false
      drawAs=Image
      imageType=NoImage
      mirroring=NoMirror
      resourceName=None
      resourceObject=null
      tiling=NoTile
      imageSize=<32 32>
      margin=<ltrb 0 0 0 0>
      outlineSettings {
       bUseBrushTransparency=false
       roundingType=HalfHeightRadius
       width=0
       color {colorUseRule=UseColor_Specified, specifiedColor=<rgba 0 0 0 0>}
       cornerRadii=<0 0 0 0>
      }
      tintColor {colorUseRule=UseColor_Specified, specifiedColor=<rgba 1 1 1 1>}
      uVRegion {bIsValid=false, max=<0 0>, min=<0 0>}
     }
    }
    renderTransformPivot=<0 0>
   }
   slot {class=SizeBoxSlot, properties {verticalAlignment=VAlign_Center}}
  }
  Image_CTABottomBar {
   widgetClassPath=&Image
   bIsVariable=true
   parent=Overlay_AlertBannerWrapper
   widget=Image_CTABottomBar
   properties {
    brush {
     bIsDynamicallyLoaded=false
     drawAs=Image
     imageType=FullColor
     mirroring=NoMirror
     resourceName=None
     resourceObject=/ToyFoxGameplay/UI/Materials/M_ToyFox_CTABottomFrame
     tiling=NoTile
     imageSize=<300 198>
     margin=<ltrb 0 0 0 0>
     outlineSettings {
      bUseBrushTransparency=false
      roundingType=HalfHeightRadius
      width=0
      color {colorUseRule=UseColor_Specified, specifiedColor=<rgba 0 0 0 0>}
      cornerRadii=<0 0 0 0>
     }
     tintColor {colorUseRule=UseColor_Specified, specifiedColor=<rgba 1 1 1 1>}
     uVRegion {bIsValid=false, max=<0 0>, min=<0 0>}
    }
    renderTransform {angle=0, scale=<0.95 1>, shear=<0 0>, translation=<-60 40>}
   }
   slot {class=OverlaySlot, properties {horizontalAlignment=HAlign_Fill, verticalAlignment=VAlign_Bottom}}
  }
  Overlay_IncomingBannerWrapper {
   widgetClassPath=&Overlay
   parent=AlertStyleSwitcher
   widget=Overlay_IncomingBannerWrapper
   slot {class=CommonVisibilitySwitcherSlot}
  }
  VerticalBox_129 {widgetClassPath=&VerticalBox, parent=Overlay_IncomingBannerWrapper, widget=VerticalBox_129, slot {class=OverlaySlot}}
  Overlay_0 {
   widgetClassPath=&Overlay
   parent=VerticalBox_129
   widget=Overlay_0
   slot {class=VerticalBoxSlot, properties {horizontalAlignment=HAlign_Left, verticalAlignment=VAlign_Top}}
  }
  Image_IncomingBacker {
   widgetClassPath=&Image
   bIsVariable=true
   parent=Overlay_0
   widget=Image_IncomingBacker
   properties {
    brush {
     bIsDynamicallyLoaded=false
     drawAs=Box
     imageType=FullColor
     mirroring=NoMirror
     resourceName=None
     resourceObject=/ToyFoxGameplay/UI/Materials/M_UI_HivePOIEnter_Backer
     tiling=NoTile
     imageSize=<700 250>
     margin=<ltrb 0.2 0.5 0.2 0.5>
     outlineSettings {
      bUseBrushTransparency=false
      roundingType=HalfHeightRadius
      width=0
      color {colorUseRule=UseColor_Specified, specifiedColor=<rgba 0 0 0 0>}
      cornerRadii=<0 0 0 0>
     }
     tintColor {colorUseRule=UseColor_Specified, specifiedColor=<rgba 1 1 1 1>}
     uVRegion {bIsValid=false, max=<0 0>, min=<0 0>}
    }
    renderTransform {angle=0, scale=<1 1>, shear=<0 0>, translation=<0 3>}
   }
   slot {class=OverlaySlot, properties {horizontalAlignment=HAlign_Fill, verticalAlignment=VAlign_Fill}}
  }
  Image_IncomingIcon {
   widgetClassPath=&Image
   bIsVariable=true
   parent=Overlay_0
   widget=Image_IncomingIcon
   properties {brush {
     bIsDynamicallyLoaded=false
     drawAs=Image
     imageType=FullColor
     mirroring=NoMirror
     resourceName=None
     resourceObject=/ToyFoxGameplay/UI/Materials/M_UI_ToyFox_WarningIcon
     tiling=NoTile
     imageSize=<600 250>
     margin=<ltrb 0 0 0 0>
     outlineSettings {
      bUseBrushTransparency=false
      roundingType=HalfHeightRadius
      width=0
      color {colorUseRule=UseColor_Specified, specifiedColor=<rgba 0 0 0 0>}
      cornerRadii=<0 0 0 0>
     }
     tintColor {colorUseRule=UseColor_Specified, specifiedColor=<rgba 1 1 1 1>}
     uVRegion {bIsValid=false, max=<0 0>, min=<0 0>}
    }}
   slot {class=OverlaySlot, properties {horizontalAlignment=HAlign_Center}}
  }
  Image_Glow {
   widgetClassPath=&Image
   bIsVariable=true
   parent=Overlay_0
   widget=Image_Glow
   properties {
    renderOpacity=0.05
    brush {
     bIsDynamicallyLoaded=false
     drawAs=Image
     imageType=FullColor
     mirroring=NoMirror
     resourceName=None
     resourceObject=/ToyFoxGameplay/UI/Materials/M_UI_ToyFox_HardEdgedGlow
     tiling=NoTile
     imageSize=<150 150>
     margin=<ltrb 0 0 0 0>
     outlineSettings {
      bUseBrushTransparency=false
      roundingType=HalfHeightRadius
      width=0
      color {colorUseRule=UseColor_Specified, specifiedColor=<rgba 0 0 0 0>}
      cornerRadii=<0 0 0 0>
     }
     tintColor {colorUseRule=UseColor_Specified, specifiedColor=<rgba 1 1 1 1>}
     uVRegion {bIsValid=false, max=<0 0>, min=<0 0>}
    }
    renderTransform {angle=0, scale=<1 1>, shear=<0 0>, translation=<0 -30>}
   }
   slot {class=OverlaySlot, properties {horizontalAlignment=HAlign_Center, verticalAlignment=VAlign_Center}}
  }
  Overlay {
   widgetClassPath=&Overlay
   parent=Overlay_0
   widget=Overlay
   slot {class=OverlaySlot, properties {horizontalAlignment=HAlign_Center, verticalAlignment=VAlign_Center}}
  }
  Spacer_65 {
   widgetClassPath=&Spacer
   parent=Overlay
   widget=Spacer_65
   properties {size=<600 1>}
   slot {class=OverlaySlot, properties {horizontalAlignment=HAlign_Fill, verticalAlignment=VAlign_Fill}}
  }
  SizeBox_0 {
   widgetClassPath=&SizeBox
   parent=Overlay
   widget=SizeBox_0
   properties {bOverride_WidthOverride=true, widthOverride=500}
   slot {class=OverlaySlot, properties {horizontalAlignment=HAlign_Center, verticalAlignment=VAlign_Center}}
  }
  Text_IncomingTitle {
   widgetClassPath=&CommonTextBlock
   bIsVariable=true
   parent=SizeBox_0
   widget=Text_IncomingTitle
   properties {
    justification=Center
    scrollStyle=&TextScrollStyle-Base_C
    text="{MEGAZORD INCOMING}"
    textTransformPolicy=ToUpper
    font {
     bForceMonospaced=false
     bMaterialIsStencil=false
     fontMaterial=null
     fontObject=&BurbankSmall
     letterSpacing=0
     monospacedWidth=1
     size=24.75
     skewAmount=0
     typefaceFontName=Black
     outlineSettings {
      bApplyOutlineToDropShadows=false
      bMiteredCorners=false
      bSeparateFillAlpha=false
      outlineMaterial=/FortUILibrary/Materials/Foundation/M-FontShadow
      outlineSize=2
      outlineColor=<rgba 0 0 0 0.8>
     }
    }
    renderTransform {angle=0, scale=<1 1>, shear=<0 0>, translation=<0 5>}
    shadowOffset=<0 0>
   }
   slot {class=SizeBoxSlot}
  }
  Image_TextFx {
   widgetClassPath=&Image
   bIsVariable=true
   parent=Overlay
   widget=Image_TextFx
   properties {
    renderOpacity=0
    brush {
     bIsDynamicallyLoaded=false
     drawAs=Image
     imageType=FullColor
     mirroring=NoMirror
     resourceName=None
     resourceObject=/ToyFoxGameplay/UI/Materials/M_UI_TextBlink
     tiling=NoTile
     imageSize=<32 32>
     margin=<ltrb 0 0 0 0>
     outlineSettings {
      bUseBrushTransparency=false
      roundingType=HalfHeightRadius
      width=0
      color {colorUseRule=UseColor_Specified, specifiedColor=<rgba 0 0 0 0>}
      cornerRadii=<0 0 0 0>
     }
     tintColor {colorUseRule=UseColor_Specified, specifiedColor=<rgba 1 1 1 1>}
     uVRegion {bIsValid=false, max=<0 0>, min=<0 0>}
    }
    renderTransform {angle=0, scale=<1.3 1.6>, shear=<0 0>, translation=<0 4>}
   }
   slot {class=OverlaySlot, properties {horizontalAlignment=HAlign_Fill, verticalAlignment=VAlign_Fill}}
  }
  Overlay_2 {
   widgetClassPath=&Overlay
   parent=VerticalBox_129
   widget=Overlay_2
   properties {renderTransform {angle=0, scale=<1.2 1.2>, shear=<0 0>, translation=<0 -70>}}
   slot {class=VerticalBoxSlot, properties {horizontalAlignment=HAlign_Center, verticalAlignment=VAlign_Top}}
  }
  Image_ThreatBacker {
   widgetClassPath=&Image
   bIsVariable=true
   parent=Overlay_2
   widget=Image_ThreatBacker
   properties {brush {
     bIsDynamicallyLoaded=false
     drawAs=Image
     imageType=FullColor
     mirroring=NoMirror
     resourceName=None
     resourceObject=/ToyFoxGameplay/UI/Materials/M_ToyFox_CTAIncoming_Sub
     tiling=NoTile
     imageSize=<50 50>
     margin=<ltrb 0 0 0 0>
     outlineSettings {
      bUseBrushTransparency=false
      roundingType=HalfHeightRadius
      width=0
      color {colorUseRule=UseColor_Specified, specifiedColor=<rgba 0 0 0 0>}
      cornerRadii=<0 0 0 0>
     }
     tintColor {colorUseRule=UseColor_Specified, specifiedColor=<rgba 1 1 1 1>}
     uVRegion {bIsValid=false, max=<0 0>, min=<0 0>}
    }}
   slot {class=OverlaySlot, properties {horizontalAlignment=HAlign_Fill, padding=<ltrb -20 0 -20 0>}}
  }
  SizeBox_1 {
   widgetClassPath=&SizeBox
   parent=Overlay_2
   widget=SizeBox_1
   properties {bOverride_MaxDesiredWidth=true, maxDesiredWidth=225, widthOverride=225}
   slot {class=OverlaySlot, properties {horizontalAlignment=HAlign_Center, verticalAlignment=VAlign_Center}}
  }
  Text_IncomingSubheader {
   widgetClassPath=&CommonTextBlock
   bIsVariable=true
   parent=SizeBox_1
   widget=Text_IncomingSubheader
   properties {
    applyLineHeightToBottomLine=false
    justification=Center
    scrollStyle=&TextScrollStyle-Base_C
    text="{GIANT BUG DETECTED}"
    textTransformPolicy=ToUpper
    colorAndOpacity {colorUseRule=UseColor_Specified, specifiedColor=<rgba 1 0.2232 0 1>}
    font {
     bForceMonospaced=false
     bMaterialIsStencil=false
     fontMaterial=null
     fontObject=&BurbankSmall
     letterSpacing=0
     monospacedWidth=1
     size=12
     skewAmount=0
     typefaceFontName=Black
     outlineSettings {
      bApplyOutlineToDropShadows=false
      bMiteredCorners=false
      bSeparateFillAlpha=false
      outlineMaterial=&M-FontShadow
      outlineSize=2
      outlineColor=<rgba 1 0.3095 0 0.502>
     }
    }
    shadowOffset=<0 0>
   }
   slot {class=SizeBoxSlot}
  }
  Image_HighThreatTextGlow {
   widgetClassPath=&Image
   bIsVariable=true
   parent=Overlay_2
   widget=Image_HighThreatTextGlow
   properties {
    renderOpacity=0
    brush {
     bIsDynamicallyLoaded=false
     drawAs=Image
     imageType=FullColor
     mirroring=NoMirror
     resourceName=None
     resourceObject=&M_UI_ToyFox_HardEdgedGlow
     tiling=NoTile
     imageSize=<203.4312 34.2083>
     margin=<ltrb 0 0 0 0>
     outlineSettings {
      bUseBrushTransparency=false
      roundingType=HalfHeightRadius
      width=0
      color {colorUseRule=UseColor_Specified, specifiedColor=<rgba 0 0 0 0>}
      cornerRadii=<0 0 0 0>
     }
     tintColor {colorUseRule=UseColor_Specified, specifiedColor=<rgba 1 1 1 1>}
     uVRegion {bIsValid=false, max=<0 0>, min=<0 0>}
    }
    renderTransform {angle=90, scale=<0.15 5>, shear=<0 0>, translation=<71 0>}
   }
   slot {class=OverlaySlot, properties {horizontalAlignment=HAlign_Center, verticalAlignment=VAlign_Center}}
  }
  TutorialHighlight {widgetClassPath=&TutorialHighlight}
 ]
}
