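# Prioritized widget wrapper for the ToyFox alert banner, built on AthenaCommonPrioritizedWidget.
# On Construct it binds the banner's alert started and finished events to set the widget's active-content state on and off.
# It also checks one tick later whether an alert is already active and, if so, marks content active.
# On Destruct it unbinds both events and clears the active-content state.
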WidgetBlueprint /ToyFoxGameplay/UI/Widgets/Alert/WBP_ToyFox_AlertBanner_Prioritized
graphs [EventGraph {nodes [
   OnAlertStarted() {is=title K2Node_CustomEvent}
   > SetHasActiveContent!(execute, self:Fort Prioritized Widget Interface, bNewActive:bool=true) -> (then)
   OnAlertFinished() {is=title K2Node_CustomEvent}
   > SetHasActiveContent(bNewActive=false)
   Event Construct()
   n1 = Create Event()
   > Bind Event to On Alert Started(
    self:WBP Toy Fox Alert Banner Object Reference=Get WBP_ToyFox_AlertBanner() -> (WBP_ToyFox_AlertBanner:WBP Toy Fox Alert Banner Object Reference)
    Delegate=@n1
   ) {is=title K2Node_AddDelegate}
   n2 = Create Event()
   > Bind Event to On Alert Finished(self:WBP Toy Fox Alert Banner Object Reference=Get WBP_ToyFox_AlertBanner(), Delegate=@n2) {is=title K2Node_AddDelegate}
   > DelayUntilNextTick!(execute) -> (then)
   > Branch(Condition=IsAlertActive!(self:WBP Toy Fox Alert Banner Object Reference=Get WBP_ToyFox_AlertBanner()) -> (bAlertActive:bool))
   > SetHasActiveContent(bNewActive=true)
   Event Destruct()
   > Unbind Event from On Alert Started(self:WBP Toy Fox Alert Banner Object Reference=Get WBP_ToyFox_AlertBanner(), Delegate=@n1) {is=title K2Node_RemoveDelegate}
   > Unbind Event from On Alert Finished(self:WBP Toy Fox Alert Banner Object Reference=Get WBP_ToyFox_AlertBanner(), Delegate=@n2) {is=title K2Node_RemoveDelegate}
   > SetHasActiveContent(bNewActive=false)
  ]}]
variables []
widgets {
 info {
  inheritedWidgetCount=1
  parentClass=/Script/FortniteUI.AthenaCommonPrioritizedWidget
  rootWidgetClass=&/WBP_ToyFox_AlertBanner.WBP_ToyFox_AlertBanner_C
  widgetCount=2
 }
 widgets [
  WBP_ToyFox_AlertBanner {widgetClassPath=&/WBP_ToyFox_AlertBanner, bIsVariable=true, widget=WBP_ToyFox_AlertBanner}
  TutorialHighlight {widgetClassPath=&TutorialHighlight}
 ]
}
