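# HUD alert banner showing a "JOINING SQUAD IN" countdown when the view target's team is interacting with a ToyFox pedestal but the view target itself is not.
# Game state component delegates supply team and progress; view target changes restart async gameplay tag queries for ToyFox.Pedestal.Interacting.
# Both feed RefreshWidgetVisibility, which plays material-driven Show/Hide animations or collapses the widget.
# While visible, Tick converts progress into the clamped countdown number.
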WidgetBlueprint /ToyFoxGameplay/UI/Widgets/Alert/WBP_ToyFox_PedestalInteractionProgress
animations [
 Show {
  binding_count=6
  length_seconds=5
  object=&self:Show
  sequence {bindings [
    OverlayTextWrapper {
     id=#1
     possessed_class=Overlay
     tracks [{
       type=MovieSceneFloatTrack
       display_name=Render Opacity
       sections [{
         type=MovieSceneFloatSection
         end_frame=28
         start_frame=0
         channels [None {type=MovieSceneScriptingFloatChannel, keys [[0,0],[4,0],[8,1],[28.6667,1]]}]
        }]
      }]
    }
    Overlay_SplashBannerWrapper {
     id=#2
     possessed_class=Overlay
     tracks [
      {
       type=MovieScene2DTransformTrack
       sections [{
         type=MovieScene2DTransformSection
         end_frame=28
         start_frame=0
         channels [
          Translation.X {type=MovieSceneScriptingFloatChannel, const [0,5]}
          Translation.Y {type=MovieSceneScriptingFloatChannel, keys [[0,0],[7.3333,0],[8,4],[10.6667,0],[28.6667,0]]}
          Angle {type=MovieSceneScriptingFloatChannel, const [0,1]}
          Scale.X {type=MovieSceneScriptingFloatChannel, const [0.8,1]}
          Scale.Y {type=MovieSceneScriptingFloatChannel, const [0.8,1]}
          Shear.X {type=MovieSceneScriptingFloatChannel, const [0,1]}
          Shear.Y {type=MovieSceneScriptingFloatChannel, const [0,1]}
         ]
        }]
      }
      {
       type=MovieSceneByteTrack
       display_name=Visibility
       sections [{type=MovieSceneByteSection, end_frame=0, start_frame=0, channels [None {type=MovieSceneScriptingByteChannel, const [4,2]}]}]
      }
      {
       type=MovieSceneFloatTrack
       display_name=Render Opacity
       sections [{type=MovieSceneFloatSection, end_frame=0, start_frame=0, channels [None {type=MovieSceneScriptingFloatChannel, const [1,1]}]}]
      }
     ]
    }
    Image_Icon {
     id=#3
     possessed_class=Image
     tracks [{
       type=MovieSceneWidgetMaterialTrack
       sections [{
         type=MovieSceneParameterSection
         end_frame=29
         start_frame=0
         channels [
          Accent Thickness Offset {type=MovieSceneScriptingFloatChannel, default=0}
          Accent Opacity {type=MovieSceneScriptingFloatChannel, keys [[0,0],[3.3333,0],[6,1],[28.6667,1]]}
          Glow Base Opacity {type=MovieSceneScriptingFloatChannel, keys [[0,0],[4.6667,0],[7.3333,1],[28.6667,1]]}
          Head Scale {type=MovieSceneScriptingFloatChannel, thinned_from=5, keys [[0,0.76],[6.6667,0.76],[7.3333,1],[28.6667,1]]}
          Head White Out {type=MovieSceneScriptingFloatChannel, thinned_from=6, keys [[0,0],[6,0],[6.6667,0.4],[8.6667,0],[28.6667,0]]}
          Glow Strength {type=MovieSceneScriptingFloatChannel, keys [[0,0.4],[5.3333,0.4],[8.6667,1],[28.6667,1]]}
          Head OffsetR {type=MovieSceneScriptingFloatChannel, default=0}
          Head OffsetG {type=MovieSceneScriptingFloatChannel, keys [[0,-0.5],[5.3333,-0.5],[7.3333,0.05],[8.6667,0],[28.6667,0]]}
          Head OffsetB {type=MovieSceneScriptingFloatChannel, default=0}
          Head OffsetA {type=MovieSceneScriptingFloatChannel, default=1}
          Glow ColorR {type=MovieSceneScriptingFloatChannel, keys [[0,1],[6,0.8976],[6.6667,1],[8.6667,0.8976],[28.6667,0.8976]]}
          Glow ColorG {type=MovieSceneScriptingFloatChannel, keys [[0,1],[6,0.0497],[6.6667,1],[8.6667,0.0497],[28.6667,0.0497]]}
          Glow ColorB {type=MovieSceneScriptingFloatChannel, keys [[0,1],[6,0.1151],[6.6667,1],[8.6667,0.1151],[28.6667,0.1151]]}
          Glow ColorA {type=MovieSceneScriptingFloatChannel, const [1,5]}
          Glow Base ColorR {type=MovieSceneScriptingFloatChannel, keys [[0,0.8976],[6,0.8976],[6.6667,1],[8.6667,0.8976],[28.6667,0.8976]]}
          Glow Base ColorG {type=MovieSceneScriptingFloatChannel, keys [[0,0.0497],[6,0.0497],[6.6667,0.2189],[8.6667,0.0497],[28.6667,0.0497]]}
          Glow Base ColorB {type=MovieSceneScriptingFloatChannel, keys [[0,0.1151],[6,0.1151],[6.6667,0.5064],[8.6667,0.1151],[28.6667,0.1151]]}
          Glow Base ColorA {type=MovieSceneScriptingFloatChannel, const [1,5]}
         ]
        }]
      }]
    }
    Image_Banner {
     id=#4
     possessed_class=Image
     tracks [{
       type=MovieSceneWidgetMaterialTrack
       sections [{
         type=MovieSceneParameterSection
         end_frame=29
         start_frame=0
         channels [
          Lightning Border - Reveal {type=MovieSceneScriptingFloatChannel, keys [[0,0],[0.6667,0],[8,1],[28.6667,1]]}
          Lightning Border - Reveal Angle {type=MovieSceneScriptingFloatChannel, thinned_from=5, keys [[0,0.1],[4,0.1],[8,-0.2],[28.6667,-0.2]]}
          Lightning Accent - Opacity {type=MovieSceneScriptingFloatChannel, keys [[0,0],[2.6667,0],[5.3333,0.02],[6.6667,0.4],[10,0.02],[28.6667,0.02]]}
          Banner - Reveal {type=MovieSceneScriptingFloatChannel, keys [[0,1],[4,1],[6.6667,0.3],[28.6667,0.3]]}
          Banner - Reveal Angle {type=MovieSceneScriptingFloatChannel, keys [[0,0],[4,0],[6,0.1],[28.6667,0.1]]}
          Banner Opacity {type=MovieSceneScriptingFloatChannel, keys [[0,0],[4,0],[5.3333,1],[28.6667,1]]}
          Accents - ColorR {type=MovieSceneScriptingFloatChannel, const [0.8714,1]}
          Accents - ColorG {type=MovieSceneScriptingFloatChannel, const [0.0561,1]}
          Accents - ColorB {type=MovieSceneScriptingFloatChannel, const [0.1651,1]}
          Accents - ColorA {type=MovieSceneScriptingFloatChannel, const [1,1]}
          Banner - Inner ColorR {type=MovieSceneScriptingFloatChannel, const [0.1094,1]}
          Banner - Inner ColorG {type=MovieSceneScriptingFloatChannel, const [0.0568,1]}
          Banner - Inner ColorB {type=MovieSceneScriptingFloatChannel, const [0.0632,1]}
          Banner - Inner ColorA {type=MovieSceneScriptingFloatChannel, const [1,1]}
          Banner - Outer ColorR {type=MovieSceneScriptingFloatChannel, const [0.0503,1]}
          Banner - Outer ColorG {type=MovieSceneScriptingFloatChannel, const [0.0137,1]}
          Banner - Outer ColorB {type=MovieSceneScriptingFloatChannel, const [0.0166,1]}
          Banner - Outer ColorA {type=MovieSceneScriptingFloatChannel, const [1,1]}
          Lightning Accent - ColorR {type=MovieSceneScriptingFloatChannel, const [0.8714,1]}
          Lightning Accent - ColorG {type=MovieSceneScriptingFloatChannel, const [0.0561,1]}
          Lightning Accent - ColorB {type=MovieSceneScriptingFloatChannel, const [0.1651,1]}
          Lightning Accent - ColorA {type=MovieSceneScriptingFloatChannel, const [1,1]}
         ]
        }]
      }]
    }
    Image_SubBanner {
     id=#5
     possessed_class=Image
     tracks [
      {
       type=MovieSceneWidgetMaterialTrack
       sections [{
         type=MovieSceneParameterSection
         end_frame=29
         start_frame=0
         channels [
          Accents - Reveal {type=MovieSceneScriptingFloatChannel, keys [[0,0.604],[14,0.604],[16,0.38],[28.6667,0.38]]}
          Accents - Reveal Angle {type=MovieSceneScriptingFloatChannel, const [0.003,1]}
          Accents - Invert Reveal? {type=MovieSceneScriptingFloatChannel, const [1,1]}
          Banner Opacity {type=MovieSceneScriptingFloatChannel, keys [[0,1],[13.3333,1],[15.3333,0.7],[28.6667,0.7]]}
          Accent ColorR {type=MovieSceneScriptingFloatChannel, thinned_from=5, keys [[0,1],[13.3333,1],[14.6667,0.0189],[28.6667,0.0189]]}
          Accent ColorG {type=MovieSceneScriptingFloatChannel, thinned_from=5, keys [[0,1],[13.3333,1],[14.6667,0.0498],[28.6667,0.0498]]}
          Accent ColorB {type=MovieSceneScriptingFloatChannel, thinned_from=5, keys [[0,1],[13.3333,1],[14.6667,0.0781],[28.6667,0.0781]]}
          Accent ColorA {type=MovieSceneScriptingFloatChannel, const [1,5]}
          Banner ColorR {type=MovieSceneScriptingFloatChannel, thinned_from=5, keys [[0,1],[13.3333,1],[14.6667,0.6172],[28.6667,0.6172]]}
          Banner ColorG {type=MovieSceneScriptingFloatChannel, thinned_from=5, keys [[0,1],[13.3333,1],[14.6667,0.8632],[28.6667,0.8632]]}
          Banner ColorB {type=MovieSceneScriptingFloatChannel, thinned_from=5, keys [[0,1],[13.3333,1],[14.6667,0.8879],[28.6667,0.8879]]}
          Banner ColorA {type=MovieSceneScriptingFloatChannel, const [1,5]}
         ]
        }]
      }
      {
       type=MovieSceneFloatTrack
       display_name=Render Opacity
       sections [{
         type=MovieSceneFloatSection
         end_frame=28
         start_frame=0
         channels [None {type=MovieSceneScriptingFloatChannel, keys [[0,0],[11.3333,0],[13.3333,1],[28.6667,1]]}]
        }]
      }
     ]
    }
    TextBlock_SubSplash {
     id=#6
     possessed_class=UEFN_TextBlock_C
     tracks [{
       type=MovieSceneFloatTrack
       display_name=Render Opacity
       sections [{
         type=MovieSceneFloatSection
         end_frame=28
         start_frame=0
         channels [None {type=MovieSceneScriptingFloatChannel, keys [[0,0],[13.3333,0],[15.3333,1],[28.6667,1]]}]
        }]
      }]
    }
   ]}
 }
 Hide {
  binding_count=5
  length_seconds=5
  object=&self:Hide
  sequence {bindings [
    Image_Icon {
     id=#7
     possessed_class=Image
     tracks [{
       type=MovieSceneWidgetMaterialTrack
       sections [{
         type=MovieSceneParameterSection
         end_frame=22
         start_frame=0
         channels [
          Accent Thickness Offset {type=MovieSceneScriptingFloatChannel, const [0,1]}
          Accent Opacity {type=MovieSceneScriptingFloatChannel, const [1,1]}
          Glow Base Opacity {type=MovieSceneScriptingFloatChannel, const [1,1]}
          Head Scale {type=MovieSceneScriptingFloatChannel, const [1,1]}
          Head White Out {type=MovieSceneScriptingFloatChannel, const [0,1]}
          Glow Strength {type=MovieSceneScriptingFloatChannel, const [1,1]}
          Head OffsetR {type=MovieSceneScriptingFloatChannel, const [0,1]}
          Head OffsetG {type=MovieSceneScriptingFloatChannel, const [0,1]}
          Head OffsetB {type=MovieSceneScriptingFloatChannel, const [0,1]}
          Head OffsetA {type=MovieSceneScriptingFloatChannel, const [1,1]}
          Glow ColorR {type=MovieSceneScriptingFloatChannel, const [0.8976,1]}
          Glow ColorG {type=MovieSceneScriptingFloatChannel, const [0.0497,1]}
          Glow ColorB {type=MovieSceneScriptingFloatChannel, const [0.1151,1]}
          Glow ColorA {type=MovieSceneScriptingFloatChannel, const [1,1]}
          Glow Base ColorR {type=MovieSceneScriptingFloatChannel, const [0.8976,1]}
          Glow Base ColorG {type=MovieSceneScriptingFloatChannel, const [0.0497,1]}
          Glow Base ColorB {type=MovieSceneScriptingFloatChannel, const [0.1151,1]}
          Glow Base ColorA {type=MovieSceneScriptingFloatChannel, const [1,1]}
         ]
        }]
      }]
    }
    Image_Banner {
     id=#8
     possessed_class=Image
     tracks [{
       type=MovieSceneWidgetMaterialTrack
       sections [{
         type=MovieSceneParameterSection
         end_frame=22
         start_frame=0
         channels [
          Lightning Border - Reveal {type=MovieSceneScriptingFloatChannel, const [1,1]}
          Lightning Border - Reveal Angle {type=MovieSceneScriptingFloatChannel, const [-0.02,1]}
          Lightning Accent - Opacity {type=MovieSceneScriptingFloatChannel, keys [[0,0.02],[12,0]]}
          Banner - Reveal {type=MovieSceneScriptingFloatChannel, const [0.3,1]}
          Banner - Reveal Angle {type=MovieSceneScriptingFloatChannel, const [0.1,1]}
          Banner Opacity {type=MovieSceneScriptingFloatChannel, keys [[0,1],[12,0]]}
          Accents - ColorR {type=MovieSceneScriptingFloatChannel, keys [[0,0.8714],[12,0]]}
          Accents - ColorG {type=MovieSceneScriptingFloatChannel, keys [[0,0.0561],[12,0]]}
          Accents - ColorB {type=MovieSceneScriptingFloatChannel, keys [[0,0.1651],[12,0]]}
          Accents - ColorA {type=MovieSceneScriptingFloatChannel, const [1,2]}
          Banner - Inner ColorR {type=MovieSceneScriptingFloatChannel, keys [[0,0.1094],[12,0]]}
          Banner - Inner ColorG {type=MovieSceneScriptingFloatChannel, keys [[0,0.0568],[12,0]]}
          Banner - Inner ColorB {type=MovieSceneScriptingFloatChannel, keys [[0,0.0632],[12,0]]}
          Banner - Inner ColorA {type=MovieSceneScriptingFloatChannel, keys [[0,1],[12,0]]}
          Banner - Outer ColorR {type=MovieSceneScriptingFloatChannel, keys [[0,0.0503],[12,0]]}
          Banner - Outer ColorG {type=MovieSceneScriptingFloatChannel, keys [[0,0.0137],[12,0]]}
          Banner - Outer ColorB {type=MovieSceneScriptingFloatChannel, keys [[0,0.0166],[12,0]]}
          Banner - Outer ColorA {type=MovieSceneScriptingFloatChannel, keys [[0,1],[12,0]]}
          Lightning Accent - ColorR {type=MovieSceneScriptingFloatChannel, keys [[0,0.8714],[12,0]]}
          Lightning Accent - ColorG {type=MovieSceneScriptingFloatChannel, keys [[0,0.0561],[12,0]]}
          Lightning Accent - ColorB {type=MovieSceneScriptingFloatChannel, keys [[0,0.1651],[12,0]]}
          Lightning Accent - ColorA {type=MovieSceneScriptingFloatChannel, const [1,2]}
         ]
        }]
      }]
    }
    TextBlock_SubSplash {
     id=#9
     possessed_class=UEFN_TextBlock_C
     tracks [{
       type=MovieSceneFloatTrack
       display_name=Render Opacity
       sections [{type=MovieSceneFloatSection, end_frame=0, start_frame=0, channels [None {type=MovieSceneScriptingFloatChannel, const [1,1]}]}]
      }]
    }
    Image_SubBanner {
     id=#10
     possessed_class=Image
     tracks [{
       type=MovieSceneWidgetMaterialTrack
       sections [{
         type=MovieSceneParameterSection
         end_frame=22
         start_frame=0
         channels [
          Accents - Reveal {type=MovieSceneScriptingFloatChannel, const [0.38,1]}
          Accents - Reveal Angle {type=MovieSceneScriptingFloatChannel, const [0.003,1]}
          Accents - Invert Reveal? {type=MovieSceneScriptingFloatChannel, const [1,1]}
          Banner Opacity {type=MovieSceneScriptingFloatChannel, const [0.7,1]}
          Accent ColorR {type=MovieSceneScriptingFloatChannel, const [0.0189,1]}
          Accent ColorG {type=MovieSceneScriptingFloatChannel, const [0.0498,1]}
          Accent ColorB {type=MovieSceneScriptingFloatChannel, const [0.0781,1]}
          Accent ColorA {type=MovieSceneScriptingFloatChannel, const [1,1]}
          Banner ColorR {type=MovieSceneScriptingFloatChannel, const [0.6172,1]}
          Banner ColorG {type=MovieSceneScriptingFloatChannel, const [0.8632,1]}
          Banner ColorB {type=MovieSceneScriptingFloatChannel, const [0.8879,1]}
          Banner ColorA {type=MovieSceneScriptingFloatChannel, const [1,1]}
         ]
        }]
      }]
    }
    Overlay_SplashBannerWrapper {
     id=#11
     possessed_class=Overlay
     tracks [
      {
       type=MovieSceneFloatTrack
       display_name=Render Opacity
       sections [{type=MovieSceneFloatSection, end_frame=22, start_frame=0, channels [None {type=MovieSceneScriptingFloatChannel, keys [[0,1],[12,0],[22,0]]}]}]
      }
      {
       type=MovieSceneByteTrack
       display_name=Visibility
       sections [{
         type=MovieSceneByteSection
         end_frame=22
         start_frame=0
         channels [None {type=MovieSceneScriptingByteChannel, keys [[0,4],[21.3333,4],[22,2]]}]
        }]
      }
     ]
    }
   ]}
 }
]
graphs [EventGraph {nodes [
   n1 = HandleViewTargetChanged!() -> (
    OutputDelegate
    then
    PlayerController:Fort Player Controller Object Reference
    OldViewTarget:Actor Object Reference
    NewViewTarget:Actor Object Reference
   ) {is=title K2Node_CustomEvent}
   > n2 = Set LatestViewTargetActor(LatestViewTargetActor:Actor Object Reference=@n1.NewViewTarget) -> (Output_Get:Actor Object Reference)
   > RefreshWidgetVisibility()
   > Is Valid(InputObject=@n2)
   > EndAction(self:Ability Async Object Reference=Get AsyncAction_QueryIsInteracting() -> (AsyncAction_QueryIsInteracting:Array of Ability Async Wait Gameplay Tag Query Object References))
   n3 = Get IsInteractingTagQuery() -> (IsInteractingTagQuery:Gameplay Tag Query Structure)
   > n4 = WaitGameplayTagQueryOnActor!(
    execute:Exec
    TargetActor:Actor Object Reference=@n2
    TagQuery:Gameplay Tag Query Structure=@n3
    TriggerCondition:EWaitGameplayTagQueryTriggerCondition Enum=WhenTrue
    bOnlyTriggerOnce:bool=false
   ) -> (then:Exec, AsyncTaskProxy:Ability Async Wait Gameplay Tag Query Object Reference, Triggered:Exec) {is=title K2Node_AsyncAction}
   > Add(
    TargetArray:Array of Ability Async Wait Gameplay Tag Query Object References=Get AsyncAction_QueryIsInteracting()
    NewItem:"Ability Async Wait Gameplay Tag Query Object Reference (by ref)"=@n4
   ) {is=title K2Node_CallArrayFunction}
   > n5 = WaitGameplayTagQueryOnActor(TargetActor=@n2, TagQuery=@n3, TriggerCondition=WhenFalse, bOnlyTriggerOnce=false)
   > Add(TargetArray=Get AsyncAction_QueryIsInteracting(), NewItem=@n5)
   Set bInteractingWithPedestal(execute=@n5.Triggered, bInteractingWithPedestal:bool=false) -> (Output_Get:bool)
   > RefreshWidgetVisibility()
   Set bInteractingWithPedestal(execute=@n4.Triggered, bInteractingWithPedestal=true)
   > RefreshWidgetVisibility()
   n6 = HandlePedestalInteractionInfoUpdated!() -> (OutputDelegate, then, Info:Toy Fox Pedestal Interaction Info Struct Structure) {is=title K2Node_CustomEvent}
   > Set LatestPedestalInteractionInfo(LatestPedestalInteractionInfo:Toy Fox Pedestal Interaction Info Struct Structure=@n6.Info) -> (Output_Get:Toy Fox Pedestal Interaction Info Struct Structure)
   > RefreshWidgetVisibility()
   RefreshWidgetVisibility() {is=title K2Node_CustomEvent}
   > n7 = Get!(execute:Exec) -> (then:Exec, else:Exec, LatestViewTargetActor:Actor Object Reference)
   > n8 = Branch(Condition=NOT Boolean(A=Get bInteractingWithPedestal() -> (bInteractingWithPedestal:bool)))
   n9 = Break Toy Fox Pedestal Interaction Info Struct!(ToyFoxPedestalInteractionInfoStruct:"Toy Fox Pedestal Interaction Info Struct Structure (by ref)"=Get LatestPedestalInteractionInfo() -> (LatestPedestalInteractionInfo:Toy Fox Pedestal Interaction Info Struct Structure)) -> (InteractingTeam_3_#12:int, InteractionProgress_6_#13:double=~) {is=title K2Node_BreakStruct}
   n10 = GetActorTeam!(Actor:Actor Object Reference=@n7) -> (ReturnValue:Byte=0)
   > n11 = Branch(Condition=AND Boolean(A="Equal (Byte)"(A:Byte=@n10, B:int=@n9) -> (ReturnValue:bool=~) {is=title K2Node_PromotableOperator}, B="Not Equal (Integer)"(A=@n9, B=-1)))
   > Branch(Condition="Equal (Enum)"(
     A:ESlateVisibility Enum=GetVisibility!(self:Widget Object Reference) -> (ReturnValue:ESlateVisibility Enum=Visible)
     B:ESlateVisibility Enum=Collapsed
    ))
   > SetVisibility(InVisibility=SelfHitTestInvisible)
   > PlayAnimation(InAnimation=Get Show() -> (Show:Widget Animation Object Reference))
   SetVisibility(execute=[@n11.else, @n12.Finished], InVisibility=Collapsed)
   n12 = Play Animation with Finished event!(
    execute:Exec=[@n8.else, @n7.else]
    Widget:User Widget Object Reference
    InAnimation:Widget Animation Object Reference=Get Hide() -> (Hide:Widget Animation Object Reference)
    StartAtTime:float=0
    NumLoopsToPlay:int=1
    PlayMode:EUMGSequencePlayMode Enum
    PlaybackSpeed:float=1
   ) -> (then:Exec, Result:Widget Animation Handle Structure, Finished:Exec) {is=title K2Node_PlayAnimation2}
   Event Construct()
   > Parent: Construct()
   > SetVisibility(InVisibility=Collapsed)
   > n13 = Sequence()
   n14 = GetGameStateAthena!(execute=@n13.then_1) -> (then, ReturnValue:Fort Game State Athena Object Reference)
   > Is Valid(InputObject=@n14)
   > n15 = GetComponentByClass!(
    execute
    self:Actor Object Reference=@n14
    ComponentClass:Actor Component Class Reference=/ToyFoxGameplay/Gameplay/B_ToyFoxGameStateComponent.B_ToyFoxGameStateComponent_C
   ) -> (then, ReturnValue:B Toy Fox Game State Component Object Reference)
   > Is Valid(InputObject=@n15)
   > Bind Event to On Pedestal Interaction Info Updated(self:B Toy Fox Game State Component Object Reference=@n15, Delegate=Create Event()) {is=title K2Node_AddDelegate}
   n16 = Cast To FortPlayerController(execute=@n13.then_0, Object=GetOwningPlayer())
   > Bind Event to On View Target Changed(self:Fort Player Controller Object Reference=@n16, Delegate=Create Event()) {is=title K2Node_AddDelegate}
   > HandleViewTargetChanged!(
    execute
    self
    PlayerController:Fort Player Controller Object Reference
    OldViewTarget:Actor Object Reference
    NewViewTarget:Actor Object Reference=GetViewTarget!(self:Controller Object Reference=@n16) -> (ReturnValue:Actor Object Reference)
   ) -> (then) {is=title K2Node_CallFunction}
   Event Destruct()
   > Parent: Destruct()
   > n17 = Sequence()
   > n18 = Cast To FortPlayerController(Object=GetOwningPlayer())
   > Unbind Event from On View Target Changed(self:Fort Player Controller Object Reference=@n18, Delegate=Create Event()) {is=title K2Node_RemoveDelegate}
   n19 = GetGameStateAthena(execute=@n17.then_1)
   > Is Valid(InputObject=@n19)
   > n20 = GetComponentByClass(self=@n19, ComponentClass=&B_ToyFoxGameStateComponent_C)
   > Is Valid(InputObject=@n20)
   > Unbind Event from On Pedestal Interaction Info Updated(self:B Toy Fox Game State Component Object Reference=@n20, Delegate=Create Event()) {is=title K2Node_RemoveDelegate}
   n21 = Event Tick()
   > Parent: Tick!(execute, MyGeometry:Geometry Structure=@n21.MyGeometry, InDeltaTime:float=@n21.InDeltaTime) -> (then) {is=title K2Node_CallParentFunction}
   > Branch(Condition="Not Equal (Enum)"!(A:ESlateVisibility Enum=GetVisibility() -> (ReturnValue=Visible), B:ESlateVisibility Enum=Collapsed) -> (ReturnValue:bool))
   n22 = float * float(
    A=Break Toy Fox Pedestal Interaction Info Struct(ToyFoxPedestalInteractionInfoStruct=Get LatestPedestalInteractionInfo()) -> (InteractionProgress_6_#13=~).InteractionProgress_6_#13
    B=10
   ) {is=title K2Node_PromotableOperator}
   n23 = Truncate(A=float / float(A=float - float(A=12, B=@n22), B=2))
   > "SetText (Text)"(
    self=Get TextBlock_SubSplash() -> (TextBlock_SubSplash:Text Block Object Reference)
    InText=To Text!(Value:int="Clamp (Integer)"(Value=@n23, Min=0, Max=10) -> (ReturnValue=0)) -> (ReturnValue:Text) {is=title K2Node_GenericToText}
   )
  ]}]
variables [
 AsyncAction_QueryIsInteracting {type="array<AbilityAsync_WaitGameplayTagQuery>", default []}
 bInteractingWithPedestal {type=boolean}
 IsInteractingTagQuery {
  type=GameplayTagQuery
  default {
   autoDescription=" ALL(  ANY( ToyFox.Pedestal.Interacting ) )"
   tokenStreamVersion=0
   userDescription=""
   queryTokenStream [0,1,5,1,1,1,0]
   tagDictionary [{tagName=ToyFox.Pedestal.Interacting}]
  }
 }
 LatestPedestalInteractionInfo {type=ToyFoxPedestalInteractionInfoStruct, default {interactingTeam=-1, interactionProgress=-1}}
 LatestViewTargetActor {type=Actor, default=null}
]
widgets {
 info {inheritedWidgetCount=1, parentClass=&FortHUDElementWidget, rootWidgetClass=&Overlay, widgetCount=13}
 widgets [
  Overlay_0 {widgetClassPath=&Overlay, widget=Overlay_0}
  Overlay_SplashBannerWrapper {
   widgetClassPath=&Overlay
   parent=Overlay_0
   widget=Overlay_SplashBannerWrapper
   properties {renderTransform {angle=0, scale=<0.8 0.8>, shear=<0 0>, translation=<0 0>}}
   slot {class=OverlaySlot, properties {horizontalAlignment=HAlign_Fill, verticalAlignment=VAlign_Fill}}
  }
  Image_Icon {
   widgetClassPath=&Image
   bIsVariable=true
   parent=Overlay_SplashBannerWrapper
   widget=Image_Icon
   properties {
    brush {
     bIsDynamicallyLoaded=false
     drawAs=Image
     imageType=FullColor
     mirroring=NoMirror
     resourceName=None
     resourceObject=/ToyFoxGameplay/UI/Materials/M_ToyFox_AlertSplash_Icon
     tiling=NoTile
     imageSize=<200 200>
     margin=<ltrb 0 0 0 0>
     outlineSettings {
      bUseBrushTransparency=false
      roundingType=HalfHeightRadius
      width=0
      color {colorUseRule=UseColor_Specified, specifiedColor=<rgba 0 0 0 0>}
      cornerRadii=<0 0 0 0>
     }
     tintColor {colorUseRule=UseColor_Specified, specifiedColor=<rgba 1 1 1 1>}
     uVRegion {bIsValid=false, max=<0 0>, min=<0 0>}
    }
    renderTransform {angle=0, scale=<1 1>, shear=<0 0>, translation=<0 -75>}
   }
   slot {class=OverlaySlot, properties {horizontalAlignment=HAlign_Center}}
  }
  CommonVisualAttachment_0 {
   widgetClassPath=&CommonVisualAttachment
   parent=Overlay_SplashBannerWrapper
   widget=CommonVisualAttachment_0
   properties {contentAnchor=<0.5 0.5>}
   slot {class=OverlaySlot, properties {horizontalAlignment=HAlign_Center, verticalAlignment=VAlign_Center}}
  }
  Overlay_StretchWrapper {
   widgetClassPath=&Overlay
   parent=CommonVisualAttachment_0
   widget=Overlay_StretchWrapper
   slot {class=SizeBoxSlot, properties {horizontalAlignment=HAlign_Center, verticalAlignment=VAlign_Center}}
  }
  Image_Banner {
   widgetClassPath=&Image
   bIsVariable=true
   parent=Overlay_StretchWrapper
   widget=Image_Banner
   properties {brush {
     bIsDynamicallyLoaded=false
     drawAs=Image
     imageType=FullColor
     mirroring=NoMirror
     resourceName=None
     resourceObject=/ToyFoxGameplay/UI/Materials/M_ToyFox_AlertSplash
     tiling=NoTile
     imageSize=<700 700>
     margin=<ltrb 0 0 0 0>
     outlineSettings {
      bUseBrushTransparency=false
      roundingType=HalfHeightRadius
      width=0
      color {colorUseRule=UseColor_Specified, specifiedColor=<rgba 0 0 0 0>}
      cornerRadii=<0 0 0 0>
     }
     tintColor {colorUseRule=UseColor_Specified, specifiedColor=<rgba 1 1 1 1>}
     uVRegion {bIsValid=false, max=<0 0>, min=<0 0>}
    }}
   slot {class=OverlaySlot, properties {horizontalAlignment=HAlign_Fill}}
  }
  OverlayTextWrapper {
   widgetClassPath=&Overlay
   parent=Overlay_StretchWrapper
   widget=OverlayTextWrapper
   slot {class=OverlaySlot, properties {horizontalAlignment=HAlign_Center, verticalAlignment=VAlign_Center}}
  }
  Text_SplashShadow {
   widgetClassPath=&CommonTextBlock
   bIsVariable=true
   parent=OverlayTextWrapper
   widget=Text_SplashShadow
   properties {
    text=JOINING SQUAD IN
    colorAndOpacity {colorUseRule=UseColor_Specified, specifiedColor=<rgba 0.283 0.0106 0.0374 1>}
    font {
     bForceMonospaced=false
     bMaterialIsStencil=false
     fontMaterial=null
     fontObject=&Burbank
     letterSpacing=50
     monospacedWidth=1
     size=42
     skewAmount=0
     typefaceFontName=BurbankRegular-Black
     outlineSettings {
      bApplyOutlineToDropShadows=false
      bMiteredCorners=false
      bSeparateFillAlpha=false
      outlineMaterial=null
      outlineSize=0
      outlineColor=<rgba 0 0 0 1>
     }
    }
    shadowOffset=<0 0>
   }
   slot {class=OverlaySlot, properties {padding=<ltrb 120 8 120 0>}}
  }
  Text_Splash {
   widgetClassPath=&CommonTextBlock
   bIsVariable=true
   parent=OverlayTextWrapper
   widget=Text_Splash
   properties {
    text=JOINING SQUAD IN
    font {
     bForceMonospaced=false
     bMaterialIsStencil=false
     fontMaterial=null
     fontObject=&Burbank
     letterSpacing=50
     monospacedWidth=1
     size=42
     skewAmount=0
     typefaceFontName=BurbankRegular-Black
     outlineSettings {
      bApplyOutlineToDropShadows=false
      bMiteredCorners=false
      bSeparateFillAlpha=false
      outlineMaterial=null
      outlineSize=0
      outlineColor=<rgba 0 0 0 1>
     }
    }
    shadowOffset=<0 0>
   }
   slot {class=OverlaySlot, properties {padding=<ltrb 120 0 120 0>}}
  }
  OverlaySubBanner {
   widgetClassPath=&Overlay
   bIsVariable=true
   parent=Overlay_SplashBannerWrapper
   widget=OverlaySubBanner
   properties {renderTransform {angle=0, scale=<1 1>, shear=<0 0>, translation=<0 70>}}
   slot {class=OverlaySlot, properties {horizontalAlignment=HAlign_Fill, verticalAlignment=VAlign_Center}}
  }
  Image_SubBanner {
   widgetClassPath=&Image
   bIsVariable=true
   parent=OverlaySubBanner
   widget=Image_SubBanner
   properties {brush {
     bIsDynamicallyLoaded=false
     drawAs=Image
     imageType=FullColor
     mirroring=NoMirror
     resourceName=None
     resourceObject=/ToyFoxGameplay/UI/Materials/M_ToyFox_AlertSplash_Sub
     tiling=NoTile
     imageSize=<500 50>
     margin=<ltrb 0 0 0 0>
     outlineSettings {
      bUseBrushTransparency=false
      roundingType=HalfHeightRadius
      width=0
      color {colorUseRule=UseColor_Specified, specifiedColor=<rgba 0 0 0 0>}
      cornerRadii=<0 0 0 0>
     }
     tintColor {colorUseRule=UseColor_Specified, specifiedColor=<rgba 1 1 1 1>}
     uVRegion {bIsValid=false, max=<0 0>, min=<0 0>}
    }}
   slot {class=OverlaySlot, properties {horizontalAlignment=HAlign_Center, verticalAlignment=VAlign_Bottom, padding=<ltrb 0 0 0 2>}}
  }
  TextBlock_SubSplash {
   widgetClassPath=&UEFN_TextBlock
   bIsVariable=true
   parent=OverlaySubBanner
   widget=TextBlock_SubSplash
   properties {
    text=3
    font {
     bForceMonospaced=false
     bMaterialIsStencil=false
     fontMaterial=null
     fontObject=&Burbank
     letterSpacing=0
     monospacedWidth=1
     size=31.5
     skewAmount=0
     typefaceFontName=Burbank-PlayerPower
     outlineSettings {
      bApplyOutlineToDropShadows=false
      bMiteredCorners=false
      bSeparateFillAlpha=false
      outlineMaterial=/Game/UI/Foundation/Materials/M-FontShadow
      outlineSize=3
      outlineColor=<rgba 0 0 0 0.5>
     }
    }
    strikeBrush {
     bIsDynamicallyLoaded=false
     drawAs=Image
     imageType=FullColor
     mirroring=NoMirror
     resourceName=None
     resourceObject=null
     tiling=NoTile
     imageSize=<32 32>
     margin=<ltrb 0 0 0 0>
     outlineSettings {
      bUseBrushTransparency=false
      roundingType=HalfHeightRadius
      width=0
      color {colorUseRule=UseColor_Specified, specifiedColor=<rgba 0 0 0 0>}
      cornerRadii=<0 0 0 0>
     }
     tintColor {colorUseRule=UseColor_Specified, specifiedColor=<rgba 1 1 1 1>}
     uVRegion {bIsValid=false, max=<0 0>, min=<0 0>}
    }
   }
   slot {class=OverlaySlot, properties {horizontalAlignment=HAlign_Center, verticalAlignment=VAlign_Center, padding=<ltrb 0 8 0 0>}}
  }
  TutorialHighlight {widgetClassPath=&TutorialHighlight}
 ]
}
